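# Material function for the AquaSynth character's hair that blends an animated watery, emissive clear-coat look into incoming material attributes.
# Hair color mixes parameter colors using Fresnel and a bounds-based gradient shaped by light direction.
# Caustics multiply panning texture pairs with noise-distorted bump-offset UVs, offset per RGB channel for chromatic splitting.
# Scrolling noise adds color variation and thresholded spots, and a camera-distance fade term is computed.
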MaterialFunction /BRCosmetics/Characters/Player/Female/Medium/Bodies/F_MED_AquaSynth/Materials/MF_AquaSynth_Hair
graph {expressions {
  MaterialExpressionAdd_1 {
   class=MaterialExpressionAdd
   inputs [
    A {from=MaterialExpressionBumpOffset_0}
    B {from=MaterialExpressionMultiply_10}
   ]
   properties {constB=0}
  }
  MaterialExpressionAdd_10 {
   class=MaterialExpressionAdd
   inputs [
    A {from=MaterialExpressionMultiply_22}
    B {from=MaterialExpressionComponentMask_6}
   ]
  }
  MaterialExpressionAdd_11 {
   class=MaterialExpressionAdd
   inputs [
    A {from=MaterialExpressionAdd_18}
    B {from=MaterialExpressionMultiply_64}
   ]
  }
  MaterialExpressionAdd_12 {
   class=MaterialExpressionAdd
   inputs [
    A {from=MaterialExpressionMultiply_27}
    B {from=MaterialExpressionNamedRerouteUsage_26}
   ]
  }
  MaterialExpressionAdd_13 {
   class=MaterialExpressionAdd
   inputs [
    A {from=MaterialExpressionMultiply_11}
    B {from=MaterialExpressionLinearInterpolate_0}
   ]
   properties {constB=0.01}
  }
  MaterialExpressionAdd_15 {
   class=MaterialExpressionAdd
   inputs [
    A {from=MaterialExpressionMultiply_4}
    B {from=MaterialExpressionLinearInterpolate_0}
   ]
   properties {constB=0.01}
  }
  MaterialExpressionAdd_18 {
   class=MaterialExpressionAdd
   inputs [
    A {from=MaterialExpressionNamedRerouteUsage_21}
    B {from=MaterialExpressionMultiply_16}
   ]
  }
  MaterialExpressionAdd_2 {
   class=MaterialExpressionAdd
   inputs [
    A {from=MaterialExpressionPanner_6}
    B {from=MaterialExpressionConstantBiasScale_0}
   ]
  }
  MaterialExpressionAdd_3 {
   class=MaterialExpressionAdd
   inputs [
    A {from=MaterialExpressionConvert_3}
    B {from=MaterialExpressionMultiply_12}
   ]
  }
  MaterialExpressionAdd_4 {
   class=MaterialExpressionAdd
   inputs [
    A {from=MaterialExpressionPanner_1}
    B {from=""}
   ]
   properties {constB=0.25}
  }
  MaterialExpressionAdd_5 {
   class=MaterialExpressionAdd
   inputs [
    A {from=MaterialExpressionPanner_3}
    B {from=""}
   ]
   properties {constB=0.25}
  }
  MaterialExpressionAdd_6 {
   class=MaterialExpressionAdd
   inputs [
    A {from=MaterialExpressionPanner_5}
    B {from=""}
   ]
   properties {constB=0.25}
  }
  MaterialExpressionAdd_8 {
   class=MaterialExpressionAdd
   inputs [
    A {from=MaterialExpressionMultiply_25}
    B {from=MaterialExpressionNamedRerouteUsage_26}
   ]
  }
  MaterialExpressionAdd_9 {
   class=MaterialExpressionAdd
   inputs [
    A {from=MaterialExpressionMultiply_28}
    B {from=MaterialExpressionNamedRerouteUsage_26}
   ]
  }
  MaterialExpressionBlendMaterialAttributes_0 {
   class=MaterialExpressionBlendMaterialAttributes
   inputs [
    A {from=MaterialExpressionGetMaterialAttributes_0, output=MaterialAttributes}
    B {from=MaterialExpressionSetMaterialAttributes_0, output=""}
    Alpha {from=MaterialExpressionNamedRerouteUsage_5, output=""}
   ]
   properties {}
  }
  MaterialExpressionBumpOffset_0 {
   class=MaterialExpressionBumpOffset
   inputs [
    Coordinate {from=MaterialExpressionTextureCoordinate_2, output=""}
    Height {from=MaterialExpressionConstant_2, output=""}
    HeightRatioInput {from="", output=""}
   ]
   properties {}
  }
  MaterialExpressionCameraPositionWS_0 {class=MaterialExpressionCameraPositionWS, inputs [], properties {}}
  MaterialExpressionCameraVectorWS_0 {class=MaterialExpressionCameraVectorWS, inputs [], properties {}}
  MaterialExpressionChannelMaskParameter_0 {
   class=MaterialExpressionChannelMaskParameter
   inputs [None {from=MaterialExpressionFunctionInput_1, output=""}]
   properties {group=AquaSynth, parameterName=TAHairMaskChannel}
  }
  MaterialExpressionComponentMask_6 {class=MaterialExpressionComponentMask, inputs [None {from=MaterialExpressionNamedRerouteUsage_13}], properties {b=true}}
  MaterialExpressionConstant2Vector_2 {class=MaterialExpressionConstant2Vector, properties {g=1, r=4}}
  MaterialExpressionConstant4Vector_0 {class=MaterialExpressionConstant4Vector, inputs [], properties {constant=<rgba 0 0.5 0 0.5>}}
  MaterialExpressionConstantBiasScale_0 {class=MaterialExpressionConstantBiasScale, inputs [None {from=MaterialExpressionSine_0}], properties {scale=0.1}}
  MaterialExpressionConstant_0 {class=MaterialExpressionConstant, properties {r=0.01}}
  MaterialExpressionConstant_2 {class=MaterialExpressionConstant, properties {r=-0.03}}
  MaterialExpressionConvert_0 {
   class=MaterialExpressionConvert
   inputs [
    None {from=MaterialExpressionMultiply_13}
    None {from=MaterialExpressionMultiply_5}
    None {from=MaterialExpressionMultiply_15}
   ]
   properties {
    nodeName=Make Vector3
    convertInputs [
     {
      type=Scalar
      expressionInput {
       expression=&self:MaterialExpressionMultiply_13
       inputName=None
       mask=0
       maskA=0
       maskB=0
       maskG=0
       maskR=0
       outputIndex=0
      }
     }
     {
      type=Scalar
      expressionInput {
       expression=&self:MaterialExpressionMultiply_5
       inputName=None
       mask=0
       maskA=0
       maskB=0
       maskG=0
       maskR=0
       outputIndex=0
      }
     }
     {
      type=Scalar
      expressionInput {
       expression=&self:MaterialExpressionMultiply_15
       inputName=None
       mask=0
       maskA=0
       maskB=0
       maskG=0
       maskR=0
       outputIndex=0
      }
     }
    ]
    convertOutputs [{type=Vector3}]
   }
  }
  MaterialExpressionConvert_1 {
   class=MaterialExpressionConvert
   inputs [
    None {from=MaterialExpressionScalarParameter_1}
    None {from=MaterialExpressionMultiply_6}
   ]
   properties {
    nodeName=Make Vector2
    convertInputs [
     {
      type=Scalar
      expressionInput {
       expression=&self:MaterialExpressionScalarParameter_1
       inputName=None
       mask=0
       maskA=0
       maskB=0
       maskG=0
       maskR=0
       outputIndex=0
      }
     }
     {
      type=Scalar
      expressionInput {
       expression=&self:MaterialExpressionMultiply_6
       inputName=None
       mask=0
       maskA=0
       maskB=0
       maskG=0
       maskR=0
       outputIndex=0
      }
     }
    ]
    convertOutputs [{type=Vector2}]
   }
  }
  MaterialExpressionConvert_3 {
   class=MaterialExpressionConvert
   inputs [None {from=MaterialExpressionDivide_3}]
   properties {
    nodeName=Break Float3
    convertInputs [{
      type=Vector3
      expressionInput {
       expression=&self:MaterialExpressionDivide_3
       inputName=None
       mask=0
       maskA=0
       maskB=0
       maskG=0
       maskR=0
       outputIndex=0
      }
     }]
    convertOutputs [{type=Scalar}, {type=Scalar}, {type=Scalar}]
   }
  }
  MaterialExpressionConvert_4 {
   class=MaterialExpressionConvert
   inputs [
    None {from=MaterialExpressionConvert_3}
    None {from=MaterialExpressionConvert_3}
   ]
   properties {
    nodeName=Make Vector2
    convertInputs [
     {
      type=Scalar
      expressionInput {
       expression=&self:MaterialExpressionConvert_3
       inputName=None
       mask=0
       maskA=0
       maskB=0
       maskG=0
       maskR=0
       outputIndex=0
      }
     }
     {
      type=Scalar
      expressionInput {
       expression=&self:MaterialExpressionConvert_3
       inputName=None
       mask=0
       maskA=0
       maskB=0
       maskG=0
       maskR=0
       outputIndex=1
      }
     }
    ]
    convertOutputs [{type=Vector2}]
   }
  }
  MaterialExpressionDivide_0 {
   class=MaterialExpressionDivide
   inputs [
    A {from=MaterialExpressionMultiply_3}
    B {from=MaterialExpressionMultiply_2}
   ]
  }
  MaterialExpressionDivide_1 {
   class=MaterialExpressionDivide
   inputs [
    A {from=MaterialExpressionSubtract_1}
    B {from=MaterialExpressionPreSkinnedLocalBounds_1, output=Extents}
   ]
  }
  MaterialExpressionDivide_2 {
   class=MaterialExpressionDivide
   inputs [
    A {from=MaterialExpressionLength_0}
    B {from=MaterialExpressionMultiply_41}
   ]
  }
  MaterialExpressionDivide_3 {
   class=MaterialExpressionDivide
   inputs [
    A {from=MaterialExpressionNamedRerouteUsage_6}
    B {from=""}
   ]
   properties {constB=8}
  }
  MaterialExpressionDotProduct_0 {
   class=MaterialExpressionDotProduct
   inputs [
    A {from=MaterialExpressionMaterialFunctionCall_21, output=Direction Vector}
    B {from=MaterialExpressionPixelNormalWS_0, output=""}
   ]
   properties {}
  }
  MaterialExpressionFresnel_0 {
   class=MaterialExpressionFresnel
   inputs [
    ExponentIn {from="", output=""}
    BaseReflectFractionIn {from="", output=""}
    Normal {from="", output=""}
   ]
   properties {exponent=0.3}
  }
  MaterialExpressionFresnel_1 {
   class=MaterialExpressionFresnel
   inputs [
    ExponentIn {from="", output=""}
    BaseReflectFractionIn {from="", output=""}
    Normal {from="", output=""}
   ]
   properties {exponent=1}
  }
  MaterialExpressionFresnel_2 {
   class=MaterialExpressionFresnel
   inputs [
    ExponentIn {from="", output=""}
    BaseReflectFractionIn {from="", output=""}
    Normal {from="", output=""}
   ]
   properties {exponent=0.5}
  }
  MaterialExpressionFresnel_4 {
   class=MaterialExpressionFresnel
   inputs [
    ExponentIn {from="", output=""}
    BaseReflectFractionIn {from="", output=""}
    Normal {from="", output=""}
   ]
   properties {exponent=0.05}
  }
  MaterialExpressionFunctionInput_0 {
   class=MaterialExpressionFunctionInput
   inputs [Preview {from="", output=""}]
   properties {inputName=OUT_MA, inputType=FunctionInput_MaterialAttributes}
  }
  MaterialExpressionFunctionInput_1 {
   class=MaterialExpressionFunctionInput
   inputs [Preview {from="", output=""}]
   properties {bUsePreviewValueAsDefault=true, inputName=TechArtMask, sortPriority=1, previewValue=<1 1 1 1>}
  }
  MaterialExpressionFunctionInput_4 {
   class=MaterialExpressionFunctionInput
   inputs [Preview {from="", output=""}]
   properties {bUsePreviewValueAsDefault=true, inputName=Mask Input, sortPriority=1}
  }
  MaterialExpressionFunctionOutput_1 {
   class=MaterialExpressionFunctionOutput
   inputs [None {from=MaterialExpressionBlendMaterialAttributes_0, output=""}]
   properties {outputName=OUT_MA}
  }
  MaterialExpressionGetMaterialAttributes_0 {
   class=MaterialExpressionGetMaterialAttributes
   inputs [None {from=MaterialExpressionFunctionInput_0, output=""}]
   properties {attributeGetTypes [
     #1
     #2
     #3
     #4
     #5
     #6
    ]}
  }
  MaterialExpressionLength_0 {class=MaterialExpressionLength, inputs [None {from=MaterialExpressionSubtract_5, output=""}], properties {}}
  MaterialExpressionLinearInterpolate_0 {
   class=MaterialExpressionLinearInterpolate
   inputs [
    A {from=""}
    B {from=MaterialExpressionScalarParameter_0}
    Alpha {from=MaterialExpressionPower_0}
   ]
   properties {constA=0.001}
  }
  MaterialExpressionLinearInterpolate_11 {
   class=MaterialExpressionLinearInterpolate
   inputs [
    A {from=""}
    B {from=""}
    Alpha {from=MaterialExpressionNamedRerouteUsage_22}
   ]
   properties {constA=2.5}
  }
  MaterialExpressionLinearInterpolate_3 {
   class=MaterialExpressionLinearInterpolate
   inputs [
    A {from=MaterialExpressionLinearInterpolate_7}
    B {from=MaterialExpressionVectorParameter_1, output=RGB}
    Alpha {from=MaterialExpressionFresnel_1}
   ]
  }
  MaterialExpressionLinearInterpolate_4 {
   class=MaterialExpressionLinearInterpolate
   inputs [
    A {from=MaterialExpressionLinearInterpolate_9}
    B {from=MaterialExpressionVectorParameter_2, output=RGB}
    Alpha {from=MaterialExpressionPower_2}
   ]
  }
  MaterialExpressionLinearInterpolate_5 {
   class=MaterialExpressionLinearInterpolate
   inputs [
    A {from=""}
    B {from=""}
    Alpha {from=MaterialExpressionNamedRerouteUsage_11}
   ]
   properties {constA=0.5, constB=2}
  }
  MaterialExpressionLinearInterpolate_7 {
   class=MaterialExpressionLinearInterpolate
   inputs [
    A {from=MaterialExpressionVectorParameter_2, output=RGB}
    B {from=MaterialExpressionVectorParameter_0, output=RGB}
    Alpha {from=MaterialExpressionNamedRerouteUsage_19}
   ]
  }
  MaterialExpressionLinearInterpolate_9 {
   class=MaterialExpressionLinearInterpolate
   inputs [
    A {from=MaterialExpressionVectorParameter_1, output=RGB}
    B {from=MaterialExpressionLinearInterpolate_3}
    Alpha {from=MaterialExpressionPower_12}
   ]
  }
  MaterialExpressionLocalPosition_1 {class=MaterialExpressionLocalPosition, inputs [], properties {localOrigin=InstancePreSkinning}}
  MaterialExpressionMaterialFunctionCall_0 {
   class=MaterialExpressionMaterialFunctionCall
   inputs [Float3 {from=MaterialExpressionSubtract_0}]
   properties {materialFunction=&BreakOutFloat3Components}
  }
  MaterialExpressionMaterialFunctionCall_1 {
   class=MaterialExpressionMaterialFunctionCall
   inputs [
    X {from=MaterialExpressionMaterialFunctionCall_0, output=R}
    Y {from=MaterialExpressionMaterialFunctionCall_0, output=R}
   ]
   properties {materialFunction=&MakeFloat2}
  }
  MaterialExpressionMaterialFunctionCall_16 {
   class=MaterialExpressionMaterialFunctionCall
   inputs [
    Mask In {from=MaterialExpressionMultiply_18}
    Amount {from=MaterialExpressionConstant4Vector_0, output=R}
    Range {from=MaterialExpressionConstant4Vector_0, output=R}
    Start Offset {from=MaterialExpressionConstant4Vector_0, output=R}
    Sharpness {from=MaterialExpressionConstant4Vector_0, output=R}
   ]
   properties {materialFunction=/Game/Global/MaterialFunctions/MF_MaskShifting}
  }
  MaterialExpressionMaterialFunctionCall_19 {
   class=MaterialExpressionMaterialFunctionCall
   inputs [Float3 {from=MaterialExpressionFunctionInput_4}]
   properties {desc="R = AO\r\nG = GMap\r\nB = Skin", materialFunction=&BreakOutFloat3Components}
  }
  MaterialExpressionMaterialFunctionCall_21 {
   class=MaterialExpressionMaterialFunctionCall
   inputs []
   properties {materialFunction=/Game/Global/Materials/Character_Parent_2023/Anime_Parent_2023/MaterialFunctions/MFC_AnimeShader_LightDirection}
  }
  MaterialExpressionMax_3 {
   class=MaterialExpressionMax
   inputs [
    A {from=MaterialExpressionSaturate_0}
    B {from=MaterialExpressionOneMinus_2}
   ]
  }
  MaterialExpressionMultiply_0 {
   class=MaterialExpressionMultiply
   inputs [
    A {from=MaterialExpressionCameraVectorWS_0}
    B {from=""}
   ]
   properties {constB=-1}
  }
  MaterialExpressionMultiply_1 {
   class=MaterialExpressionMultiply
   inputs [
    A {from=MaterialExpressionPower_1}
    B {from=""}
   ]
   properties {constB=8}
  }
  MaterialExpressionMultiply_10 {
   class=MaterialExpressionMultiply
   inputs [
    A {from=MaterialExpressionTextureSample_0, output=R}
    B {from=MaterialExpressionScalarParameter_25}
   ]
   properties {constB=5}
  }
  MaterialExpressionMultiply_11 {
   class=MaterialExpressionMultiply
   inputs [
    A {from=MaterialExpressionAdd_1}
    B {from=MaterialExpressionMultiply_7}
   ]
  }
  MaterialExpressionMultiply_12 {
   class=MaterialExpressionMultiply
   inputs [
    A {from=MaterialExpressionTime_1}
    B {from=""}
   ]
   properties {constB=0.6666}
  }
  MaterialExpressionMultiply_13 {
   class=MaterialExpressionMultiply
   inputs [
    A {from=MaterialExpressionTextureSample_3, output=R}
    B {from=MaterialExpressionTextureSample_4, output=R}
   ]
  }
  MaterialExpressionMultiply_14 {
   class=MaterialExpressionMultiply
   inputs [
    A {from=MaterialExpressionNamedRerouteUsage_21}
    B {from=MaterialExpressionNamedRerouteUsage_20}
   ]
  }
  MaterialExpressionMultiply_15 {
   class=MaterialExpressionMultiply
   inputs [
    A {from=MaterialExpressionTextureSample_6, output=R}
    B {from=MaterialExpressionTextureSample_7, output=R}
   ]
  }
  MaterialExpressionMultiply_16 {
   class=MaterialExpressionMultiply
   inputs [
    A {from=MaterialExpressionMultiply_14}
    B {from=MaterialExpressionLinearInterpolate_11}
   ]
   properties {constB=10}
  }
  MaterialExpressionMultiply_18 {
   class=MaterialExpressionMultiply
   inputs [
    A {from=MaterialExpressionAdd_10}
    B {from=MaterialExpressionMax_3}
   ]
  }
  MaterialExpressionMultiply_2 {
   class=MaterialExpressionMultiply
   inputs [
    A {from=MaterialExpressionMaterialFunctionCall_0, output=B}
    B {from=""}
   ]
   properties {constB=-1}
  }
  MaterialExpressionMultiply_22 {
   class=MaterialExpressionMultiply
   inputs [
    A {from=MaterialExpressionNamedRerouteUsage_18}
    B {from=MaterialExpressionComponentMask_6}
   ]
  }
  MaterialExpressionMultiply_25 {
   class=MaterialExpressionMultiply
   inputs [
    A {from=MaterialExpressionTextureCoordinate_0}
    B {from=""}
   ]
   properties {constB=12}
  }
  MaterialExpressionMultiply_26 {
   class=MaterialExpressionMultiply
   inputs [
    A {from=MaterialExpressionMultiply_29}
    B {from=MaterialExpressionPower_15}
   ]
  }
  MaterialExpressionMultiply_27 {
   class=MaterialExpressionMultiply
   inputs [
    A {from=MaterialExpressionTextureCoordinate_0}
    B {from=""}
   ]
   properties {constB=3}
  }
  MaterialExpressionMultiply_28 {
   class=MaterialExpressionMultiply
   inputs [
    A {from=MaterialExpressionTextureCoordinate_0}
    B {from=""}
   ]
   properties {constB=6}
  }
  MaterialExpressionMultiply_29 {
   class=MaterialExpressionMultiply
   inputs [
    A {from=MaterialExpressionTextureSample_11, output=R}
    B {from=MaterialExpressionTextureSample_13, output=R}
   ]
  }
  MaterialExpressionMultiply_3 {
   class=MaterialExpressionMultiply
   inputs [
    A {from=MaterialExpressionMaterialFunctionCall_1, output=Result}
    B {from=MaterialExpressionConstant_0}
   ]
  }
  MaterialExpressionMultiply_33 {
   class=MaterialExpressionMultiply
   inputs [
    A {from=MaterialExpressionStep_0}
    B {from=MaterialExpressionNamedRerouteUsage_7}
   ]
  }
  MaterialExpressionMultiply_34 {
   class=MaterialExpressionMultiply
   inputs [
    A {from=MaterialExpressionNamedRerouteUsage_28}
    B {from=""}
   ]
   properties {constB=0.1}
  }
  MaterialExpressionMultiply_38 {
   class=MaterialExpressionMultiply
   inputs [
    A {from=MaterialExpressionTextureCoordinate_6}
    B {from=MaterialExpressionConstant2Vector_2}
   ]
   properties {constB=4}
  }
  MaterialExpressionMultiply_4 {
   class=MaterialExpressionMultiply
   inputs [
    A {from=MaterialExpressionAdd_1}
    B {from=MaterialExpressionScalarParameter_2}
   ]
   properties {constB=0.5}
  }
  MaterialExpressionMultiply_41 {
   class=MaterialExpressionMultiply
   inputs [
    A {from=MaterialExpressionScalarParameter_9}
    B {from=""}
   ]
   properties {constB=100}
  }
  MaterialExpressionMultiply_5 {
   class=MaterialExpressionMultiply
   inputs [
    A {from=MaterialExpressionTextureSample_1, output=R}
    B {from=MaterialExpressionTextureSample_2, output=R}
   ]
  }
  MaterialExpressionMultiply_6 {
   class=MaterialExpressionMultiply
   inputs [
    A {from=MaterialExpressionScalarParameter_1}
    B {from=""}
   ]
   properties {constB=-1}
  }
  MaterialExpressionMultiply_63 {
   class=MaterialExpressionMultiply
   inputs [
    A {from=MaterialExpressionNamedRerouteUsage_21}
    B {from=MaterialExpressionNamedRerouteUsage_17}
   ]
  }
  MaterialExpressionMultiply_64 {
   class=MaterialExpressionMultiply
   inputs [
    A {from=MaterialExpressionMultiply_63}
    B {from=""}
   ]
   properties {constB=2}
  }
  MaterialExpressionMultiply_7 {
   class=MaterialExpressionMultiply
   inputs [
    A {from=MaterialExpressionScalarParameter_2}
    B {from=""}
   ]
   properties {constB=2}
  }
  MaterialExpressionMultiply_8 {
   class=MaterialExpressionMultiply
   inputs [
    A {from=MaterialExpressionPower_3}
    B {from=MaterialExpressionLinearInterpolate_5}
   ]
   properties {constB=0.5}
  }
  MaterialExpressionNamedRerouteDeclaration_0 {
   class=MaterialExpressionNamedRerouteDeclaration
   inputs [None {from=MaterialExpressionMultiply_34}]
   properties {name=OUT_BaseColor, nodeColor=<rgba 0 0.8856 1 1>}
  }
  MaterialExpressionNamedRerouteDeclaration_1 {
   class=MaterialExpressionNamedRerouteDeclaration
   inputs [None {from=MaterialExpressionNamedRerouteDeclaration_7}]
   properties {name=OUT_Metallic, nodeColor=<rgba 0.3462 1 0.0006 1>}
  }
  MaterialExpressionNamedRerouteDeclaration_10 {
   class=MaterialExpressionNamedRerouteDeclaration
   inputs [None {from=MaterialExpressionGetMaterialAttributes_0, output=Normal}]
   properties {name=IN_Normal, nodeColor=<rgba 0.25 0.25 1 1>}
  }
  MaterialExpressionNamedRerouteDeclaration_11 {
   class=MaterialExpressionNamedRerouteDeclaration
   inputs [None {from=MaterialExpressionMultiply_8}]
   properties {name=Caustics, nodeColor=<rgba 1 0 0.8437 1>}
  }
  MaterialExpressionNamedRerouteDeclaration_12 {
   class=MaterialExpressionNamedRerouteDeclaration
   inputs [None {from=MaterialExpressionSaturate_5}]
   properties {name=DistToFade, nodeColor=<rgba 0.4531 1 0 1>}
  }
  MaterialExpressionNamedRerouteDeclaration_13 {
   class=MaterialExpressionNamedRerouteDeclaration
   inputs [None {from=MaterialExpressionDivide_1}]
   properties {name=Bounding, nodeColor=<rgba 0 0.5312 1 1>}
  }
  MaterialExpressionNamedRerouteDeclaration_14 {
   class=MaterialExpressionNamedRerouteDeclaration
   inputs [None {from=MaterialExpressionMaterialFunctionCall_19, output=R}]
   properties {name=InPut Ambient Occlusion, nodeColor=<rgba 0.9453 1 0 1>}
  }
  MaterialExpressionNamedRerouteDeclaration_15 {
   class=MaterialExpressionNamedRerouteDeclaration
   inputs [None {from=MaterialExpressionShaderStageSwitch_12}]
   properties {name=PreSkinnedLoc, nodeColor=<rgba 0.7001 0 1 1>}
  }
  MaterialExpressionNamedRerouteDeclaration_16 {
   class=MaterialExpressionNamedRerouteDeclaration
   inputs [None {from=MaterialExpressionSaturate_0}]
   properties {name=DirectionLightMask, nodeColor=<rgba 0.5937 1 0 1>}
  }
  MaterialExpressionNamedRerouteDeclaration_17 {
   class=MaterialExpressionNamedRerouteDeclaration
   inputs [None {from=MaterialExpressionDivide_0}]
   properties {name=ParallexOffset, nodeColor=<rgba 0.1953 1 0 1>}
  }
  MaterialExpressionNamedRerouteDeclaration_18 {
   class=MaterialExpressionNamedRerouteDeclaration
   inputs [None {from=MaterialExpressionMultiply_33}]
   properties {name=Spots, nodeColor=<rgba 0 1 0.8594 1>}
  }
  MaterialExpressionNamedRerouteDeclaration_19 {
   class=MaterialExpressionNamedRerouteDeclaration
   inputs [None {from=MaterialExpressionSaturate_3}]
   properties {name=HairGraident, nodeColor=<rgba 1 0 0.1875 1>}
  }
  MaterialExpressionNamedRerouteDeclaration_2 {
   class=MaterialExpressionNamedRerouteDeclaration
   inputs [None {from=MaterialExpressionNamedRerouteDeclaration_8}]
   properties {name=OUT_Specular, nodeColor=<rgba 0.25 1 0 1>}
  }
  MaterialExpressionNamedRerouteDeclaration_22 {
   class=MaterialExpressionNamedRerouteDeclaration
   inputs [None {from=MaterialExpressionSaturate_4}]
   properties {name=HairColor, nodeColor=<rgba 0 0.5078 1 1>}
  }
  MaterialExpressionNamedRerouteDeclaration_23 {
   class=MaterialExpressionNamedRerouteDeclaration
   inputs [None {from=MaterialExpressionMultiply_10}]
   properties {name=Distortion, nodeColor=<rgba 0.4531 1 0 1>}
  }
  MaterialExpressionNamedRerouteDeclaration_24 {
   class=MaterialExpressionNamedRerouteDeclaration
   inputs [None {from=MaterialExpressionChannelMaskParameter_0, output=R}]
   properties {name=TAMask, nodeColor=<rgba 0.8516 0 1 1>}
  }
  MaterialExpressionNamedRerouteDeclaration_3 {
   class=MaterialExpressionNamedRerouteDeclaration
   inputs [None {from=MaterialExpressionNamedRerouteDeclaration_9}]
   properties {name=OUT_Roughness, nodeColor=<rgba 0.25 1 0 1>}
  }
  MaterialExpressionNamedRerouteDeclaration_32 {
   class=MaterialExpressionNamedRerouteDeclaration
   inputs [None {from=MaterialExpressionMultiply_1}]
   properties {name=ColorNoise, nodeColor=<rgba 1 0 0.7031 1>}
  }
  MaterialExpressionNamedRerouteDeclaration_4 {
   class=MaterialExpressionNamedRerouteDeclaration
   inputs [None {from=MaterialExpressionMaterialFunctionCall_19, output=G}]
   properties {name=Gmap, nodeColor=<rgba 1 0.375 0 1>}
  }
  MaterialExpressionNamedRerouteDeclaration_5 {
   class=MaterialExpressionNamedRerouteDeclaration
   inputs [None {from=MaterialExpressionAdd_11}]
   properties {name=OUT_EmissiveColor, nodeColor=<rgba 0 0.8 1 1>}
  }
  MaterialExpressionNamedRerouteDeclaration_6 {
   class=MaterialExpressionNamedRerouteDeclaration
   inputs [None {from=MaterialExpressionGetMaterialAttributes_0, output=BaseColor}]
   properties {name=IN_BaseColor, nodeColor=<rgba 0 0.8 1 1>}
  }
  MaterialExpressionNamedRerouteDeclaration_7 {class=MaterialExpressionNamedRerouteDeclaration, inputs [None {from=""}], properties {name=IN_Metallic, nodeColor=<rgba 0.25 1 0 1>}}
  MaterialExpressionNamedRerouteDeclaration_8 {class=MaterialExpressionNamedRerouteDeclaration, inputs [None {from=""}], properties {name=IN_Specular, nodeColor=<rgba 0.25 1 0 1>}}
  MaterialExpressionNamedRerouteDeclaration_9 {class=MaterialExpressionNamedRerouteDeclaration, inputs [None {from=""}], properties {name=IN_Roughness, nodeColor=<rgba 0.25 1 0 1>}}
  MaterialExpressionNamedRerouteUsage_1 {class=MaterialExpressionNamedRerouteUsage}
  MaterialExpressionNamedRerouteUsage_10 {class=MaterialExpressionNamedRerouteUsage}
  MaterialExpressionNamedRerouteUsage_11 {class=MaterialExpressionNamedRerouteUsage}
  MaterialExpressionNamedRerouteUsage_12 {class=MaterialExpressionNamedRerouteUsage}
  MaterialExpressionNamedRerouteUsage_13 {class=MaterialExpressionNamedRerouteUsage}
  MaterialExpressionNamedRerouteUsage_15 {class=MaterialExpressionNamedRerouteUsage}
  MaterialExpressionNamedRerouteUsage_17 {class=MaterialExpressionNamedRerouteUsage}
  MaterialExpressionNamedRerouteUsage_18 {class=MaterialExpressionNamedRerouteUsage}
  MaterialExpressionNamedRerouteUsage_19 {class=MaterialExpressionNamedRerouteUsage}
  MaterialExpressionNamedRerouteUsage_2 {class=MaterialExpressionNamedRerouteUsage}
  MaterialExpressionNamedRerouteUsage_20 {class=MaterialExpressionNamedRerouteUsage}
  MaterialExpressionNamedRerouteUsage_21 {class=MaterialExpressionNamedRerouteUsage}
  MaterialExpressionNamedRerouteUsage_22 {class=MaterialExpressionNamedRerouteUsage}
  MaterialExpressionNamedRerouteUsage_25 {class=MaterialExpressionNamedRerouteUsage}
  MaterialExpressionNamedRerouteUsage_26 {class=MaterialExpressionNamedRerouteUsage}
  MaterialExpressionNamedRerouteUsage_27 {class=MaterialExpressionNamedRerouteUsage}
  MaterialExpressionNamedRerouteUsage_28 {class=MaterialExpressionNamedRerouteUsage}
  MaterialExpressionNamedRerouteUsage_3 {class=MaterialExpressionNamedRerouteUsage}
  MaterialExpressionNamedRerouteUsage_4 {class=MaterialExpressionNamedRerouteUsage}
  MaterialExpressionNamedRerouteUsage_5 {class=MaterialExpressionNamedRerouteUsage}
  MaterialExpressionNamedRerouteUsage_6 {class=MaterialExpressionNamedRerouteUsage}
  MaterialExpressionNamedRerouteUsage_7 {class=MaterialExpressionNamedRerouteUsage}
  MaterialExpressionOneMinus_0 {class=MaterialExpressionOneMinus, inputs [None {from=MaterialExpressionFresnel_0}]}
  MaterialExpressionOneMinus_2 {class=MaterialExpressionOneMinus, inputs [None {from=MaterialExpressionPower_9}]}
  MaterialExpressionPanner_0 {
   class=MaterialExpressionPanner
   inputs [
    Coordinate {from=MaterialExpressionMultiply_11}
    Time {from=""}
    Speed {from=MaterialExpressionConvert_1}
   ]
   properties {speedX=0.06, speedY=-0.06}
  }
  MaterialExpressionPanner_1 {
   class=MaterialExpressionPanner
   inputs [
    Coordinate {from=MaterialExpressionMultiply_4}
    Time {from=""}
    Speed {from=MaterialExpressionMultiply_6}
   ]
   properties {speedX=-0.06, speedY=-0.06}
  }
  MaterialExpressionPanner_10 {
   class=MaterialExpressionPanner
   inputs [
    Coordinate {from=MaterialExpressionAdd_9}
    Time {from=""}
    Speed {from=""}
   ]
   properties {speedY=-0.04}
  }
  MaterialExpressionPanner_2 {
   class=MaterialExpressionPanner
   inputs [
    Coordinate {from=MaterialExpressionSubtract_2}
    Time {from=""}
    Speed {from=MaterialExpressionConvert_1}
   ]
   properties {speedX=0.06, speedY=-0.06}
  }
  MaterialExpressionPanner_3 {
   class=MaterialExpressionPanner
   inputs [
    Coordinate {from=MaterialExpressionSubtract_4}
    Time {from=""}
    Speed {from=MaterialExpressionMultiply_6}
   ]
   properties {speedX=-0.06, speedY=-0.06}
  }
  MaterialExpressionPanner_4 {
   class=MaterialExpressionPanner
   inputs [
    Coordinate {from=MaterialExpressionAdd_13}
    Time {from=""}
    Speed {from=MaterialExpressionConvert_1}
   ]
   properties {speedX=0.06, speedY=-0.06}
  }
  MaterialExpressionPanner_5 {
   class=MaterialExpressionPanner
   inputs [
    Coordinate {from=MaterialExpressionAdd_15}
    Time {from=""}
    Speed {from=MaterialExpressionMultiply_6}
   ]
   properties {speedX=-0.06, speedY=-0.06}
  }
  MaterialExpressionPanner_6 {
   class=MaterialExpressionPanner
   inputs [
    Coordinate {from=MaterialExpressionConvert_4}
    Time {from=""}
    Speed {from=""}
   ]
   properties {speedX=0.05, speedY=0.02}
  }
  MaterialExpressionPanner_7 {
   class=MaterialExpressionPanner
   inputs [
    Coordinate {from=MaterialExpressionMultiply_38}
    Time {from=""}
    Speed {from=""}
   ]
   properties {speedY=-0.1}
  }
  MaterialExpressionPanner_8 {
   class=MaterialExpressionPanner
   inputs [
    Coordinate {from=MaterialExpressionAdd_12}
    Time {from=""}
    Speed {from=""}
   ]
   properties {speedY=-0.02}
  }
  MaterialExpressionPanner_9 {
   class=MaterialExpressionPanner
   inputs [
    Coordinate {from=MaterialExpressionAdd_8}
    Time {from=""}
    Speed {from=""}
   ]
   properties {speedY=-0.02}
  }
  MaterialExpressionPixelNormalWS_0 {class=MaterialExpressionPixelNormalWS, inputs [], properties {}}
  MaterialExpressionPower_0 {
   class=MaterialExpressionPower
   inputs [
    Base {from=MaterialExpressionNamedRerouteUsage_10}
    Exp {from=MaterialExpressionOneMinus_0}
   ]
  }
  MaterialExpressionPower_1 {
   class=MaterialExpressionPower
   inputs [
    Base {from=MaterialExpressionTextureSample_8, output=R}
    Exp {from=""}
   ]
  }
  MaterialExpressionPower_12 {
   class=MaterialExpressionPower
   inputs [
    Base {from=MaterialExpressionNamedRerouteUsage_15}
    Exp {from=""}
   ]
   properties {constExponent=2.5}
  }
  MaterialExpressionPower_15 {
   class=MaterialExpressionPower
   inputs [
    Base {from=MaterialExpressionTextureSample_12, output=R}
    Exp {from=""}
   ]
   properties {constExponent=1.25}
  }
  MaterialExpressionPower_2 {
   class=MaterialExpressionPower
   inputs [
    Base {from=MaterialExpressionFresnel_2}
    Exp {from=""}
   ]
   properties {constExponent=3}
  }
  MaterialExpressionPower_3 {
   class=MaterialExpressionPower
   inputs [
    Base {from=MaterialExpressionConvert_0}
    Exp {from=""}
   ]
   properties {constExponent=0.6}
  }
  MaterialExpressionPower_4 {
   class=MaterialExpressionPower
   inputs [
    Base {from=MaterialExpressionLinearInterpolate_4}
    Exp {from=""}
   ]
   properties {constExponent=2.2}
  }
  MaterialExpressionPower_5 {
   class=MaterialExpressionPower
   inputs [
    Base {from=MaterialExpressionDivide_2}
    Exp {from=""}
   ]
   properties {constExponent=4}
  }
  MaterialExpressionPower_9 {
   class=MaterialExpressionPower
   inputs [
    Base {from=MaterialExpressionFresnel_4}
    Exp {from=""}
   ]
   properties {constExponent=4}
  }
  MaterialExpressionPreSkinnedLocalBounds_1 {class=MaterialExpressionPreSkinnedLocalBounds, inputs [], properties {}}
  MaterialExpressionSaturate_0 {class=MaterialExpressionSaturate, inputs [None {from=MaterialExpressionDotProduct_0}]}
  MaterialExpressionSaturate_3 {class=MaterialExpressionSaturate, inputs [None {from=MaterialExpressionMaterialFunctionCall_16, output=Result}]}
  MaterialExpressionSaturate_4 {class=MaterialExpressionSaturate, inputs [None {from=MaterialExpressionPower_4}]}
  MaterialExpressionSaturate_5 {class=MaterialExpressionSaturate, inputs [None {from=MaterialExpressionPower_5}]}
  MaterialExpressionScalarParameter_0 {class=MaterialExpressionScalarParameter, properties {defaultValue=0.01, group=AquaSynth, parameterName=Caustic_OffsetScale}}
  MaterialExpressionScalarParameter_1 {class=MaterialExpressionScalarParameter, properties {defaultValue=0.02, group=AquaSynth, parameterName=Caustic_Speed}}
  MaterialExpressionScalarParameter_2 {class=MaterialExpressionScalarParameter, properties {defaultValue=0.5, group=AquaSynth, parameterName=Caustic_Tilling}}
  MaterialExpressionScalarParameter_25 {class=MaterialExpressionScalarParameter, properties {defaultValue=0.02, group=AquaSynth, parameterName=Caustics_Distortion}}
  MaterialExpressionScalarParameter_9 {class=MaterialExpressionScalarParameter, properties {defaultValue=2.5, group=AquaSynth, parameterName=CausticsDistanceStartFade}}
  MaterialExpressionSetMaterialAttributes_0 {
   class=MaterialExpressionSetMaterialAttributes
   inputs [
    MaterialAttributes {from=MaterialExpressionGetMaterialAttributes_0, output=MaterialAttributes}
    BaseColor {from=MaterialExpressionNamedRerouteUsage_25, output=""}
    Metallic {from=MaterialExpressionGetMaterialAttributes_0, output=MaterialAttributes}
    Specular {from=MaterialExpressionGetMaterialAttributes_0, output=MaterialAttributes}
    Roughness {from=MaterialExpressionGetMaterialAttributes_0, output=MaterialAttributes}
    Normal {from=MaterialExpressionGetMaterialAttributes_0, output=MaterialAttributes}
    EmissiveColor {from=MaterialExpressionNamedRerouteUsage_4, output=""}
    ShadingModel {from=MaterialExpressionShadingModel_0, output=""}
   ]
   properties {attributeSetTypes [
     #1
     #2
     #3
     #4
     #5
     #6
     #7
    ]}
  }
  MaterialExpressionShaderStageSwitch_12 {
   class=MaterialExpressionShaderStageSwitch
   inputs [
    PixelShader {from=MaterialExpressionVertexInterpolator_16, output=PS}
    VertexShader {from=MaterialExpressionLocalPosition_1, output=XYZ}
   ]
   properties {}
  }
  MaterialExpressionShadingModel_0 {class=MaterialExpressionShadingModel, inputs [], properties {shadingModel=MSM_ClearCoat}}
  MaterialExpressionSine_0 {class=MaterialExpressionSine, inputs [None {from=MaterialExpressionAdd_3}]}
  MaterialExpressionStep_0 {
   class=MaterialExpressionStep
   inputs [
    Y {from="", output=""}
    X {from=MaterialExpressionMultiply_26, output=""}
   ]
   properties {constY=0.005}
  }
  MaterialExpressionSubtract_0 {
   class=MaterialExpressionSubtract
   inputs [
    A {from=MaterialExpressionTransform_2}
    B {from=MaterialExpressionTransform_0}
   ]
  }
  MaterialExpressionSubtract_1 {
   class=MaterialExpressionSubtract
   inputs [
    A {from=MaterialExpressionTransformPosition_0}
    B {from=MaterialExpressionPreSkinnedLocalBounds_1, output=Min}
   ]
  }
  MaterialExpressionSubtract_2 {
   class=MaterialExpressionSubtract
   inputs [
    A {from=MaterialExpressionMultiply_11}
    B {from=MaterialExpressionLinearInterpolate_0}
   ]
   properties {constB=0.01}
  }
  MaterialExpressionSubtract_4 {
   class=MaterialExpressionSubtract
   inputs [
    A {from=MaterialExpressionMultiply_4}
    B {from=MaterialExpressionLinearInterpolate_0}
   ]
   properties {constB=0.01}
  }
  MaterialExpressionSubtract_5 {
   class=MaterialExpressionSubtract
   inputs [
    A {from=MaterialExpressionNamedRerouteUsage_27}
    B {from=MaterialExpressionTransformPosition_2}
   ]
  }
  MaterialExpressionTextureCoordinate_0 {class=MaterialExpressionTextureCoordinate}
  MaterialExpressionTextureCoordinate_2 {class=MaterialExpressionTextureCoordinate}
  MaterialExpressionTextureCoordinate_6 {class=MaterialExpressionTextureCoordinate}
  MaterialExpressionTextureSample_0 {
   class=MaterialExpressionTextureSample
   inputs [
    UVs {from=MaterialExpressionAdd_2}
    Tex {from=""}
    Apply View MipBias {from=""}
   ]
   properties {samplerSource=SSM_Wrap_WorldGroupSettings, samplerType=SAMPLERTYPE_Grayscale, texture=&LowResBlurredNoise}
  }
  MaterialExpressionTextureSample_1 {
   class=MaterialExpressionTextureSample
   inputs [
    UVs {from=MaterialExpressionPanner_0}
    Tex {from=""}
    Apply View MipBias {from=""}
   ]
   properties {samplerSource=SSM_Wrap_WorldGroupSettings, texture=/Game/Global/Textures/Default/Caustics/T_Caustics_Softer}
  }
  MaterialExpressionTextureSample_11 {
   class=MaterialExpressionTextureSample
   inputs [
    UVs {from=MaterialExpressionPanner_9}
    Tex {from=""}
    Apply View MipBias {from=""}
   ]
   properties {samplerSource=SSM_Wrap_WorldGroupSettings, texture=/Game/Global/Textures/Default/Noise/T_TileNoise_02}
  }
  MaterialExpressionTextureSample_12 {
   class=MaterialExpressionTextureSample
   inputs [
    UVs {from=MaterialExpressionPanner_8}
    Tex {from=""}
    Apply View MipBias {from=""}
   ]
   properties {samplerSource=SSM_Wrap_WorldGroupSettings, samplerType=SAMPLERTYPE_Grayscale, texture=&LowResBlurredNoise}
  }
  MaterialExpressionTextureSample_13 {
   class=MaterialExpressionTextureSample
   inputs [
    UVs {from=MaterialExpressionPanner_10}
    Tex {from=""}
    Apply View MipBias {from=""}
   ]
   properties {samplerSource=SSM_Wrap_WorldGroupSettings, texture=&T_TileNoise_02}
  }
  MaterialExpressionTextureSample_2 {
   class=MaterialExpressionTextureSample
   inputs [
    UVs {from=MaterialExpressionAdd_4}
    Tex {from=""}
    Apply View MipBias {from=""}
   ]
   properties {samplerSource=SSM_Wrap_WorldGroupSettings, texture=&T_Caustics_Softer}
  }
  MaterialExpressionTextureSample_3 {
   class=MaterialExpressionTextureSample
   inputs [
    UVs {from=MaterialExpressionPanner_2}
    Tex {from=""}
    Apply View MipBias {from=""}
   ]
   properties {samplerSource=SSM_Wrap_WorldGroupSettings, texture=&T_Caustics_Softer}
  }
  MaterialExpressionTextureSample_4 {
   class=MaterialExpressionTextureSample
   inputs [
    UVs {from=MaterialExpressionAdd_5}
    Tex {from=""}
    Apply View MipBias {from=""}
   ]
   properties {samplerSource=SSM_Wrap_WorldGroupSettings, texture=&T_Caustics_Softer}
  }
  MaterialExpressionTextureSample_6 {
   class=MaterialExpressionTextureSample
   inputs [
    UVs {from=MaterialExpressionPanner_4}
    Tex {from=""}
    Apply View MipBias {from=""}
   ]
   properties {samplerSource=SSM_Wrap_WorldGroupSettings, texture=&T_Caustics_Softer}
  }
  MaterialExpressionTextureSample_7 {
   class=MaterialExpressionTextureSample
   inputs [
    UVs {from=MaterialExpressionAdd_6}
    Tex {from=""}
    Apply View MipBias {from=""}
   ]
   properties {samplerSource=SSM_Wrap_WorldGroupSettings, texture=&T_Caustics_Softer}
  }
  MaterialExpressionTextureSample_8 {
   class=MaterialExpressionTextureSample
   inputs [
    UVs {from=MaterialExpressionPanner_7}
    Tex {from=""}
    Apply View MipBias {from=""}
   ]
   properties {samplerSource=SSM_Wrap_WorldGroupSettings, samplerType=SAMPLERTYPE_Grayscale, texture=&LowResBlurredNoise}
  }
  MaterialExpressionTime_1 {class=MaterialExpressionTime}
  MaterialExpressionTransformPosition_0 {
   class=MaterialExpressionTransformPosition
   inputs [None {from=MaterialExpressionNamedRerouteUsage_12, output=""}]
   properties {transformType=TRANSFORMPOSSOURCE_Instance}
  }
  MaterialExpressionTransformPosition_2 {
   class=MaterialExpressionTransformPosition
   inputs [None {from=MaterialExpressionCameraPositionWS_0, output=""}]
   properties {transformSourceType=TRANSFORMPOSSOURCE_World}
  }
  MaterialExpressionTransform_0 {
   class=MaterialExpressionTransform
   inputs [None {from=MaterialExpressionVertexNormalWS_0}]
   properties {transformSourceType=TRANSFORMSOURCE_World, transformType=TRANSFORM_Tangent}
  }
  MaterialExpressionTransform_2 {
   class=MaterialExpressionTransform
   inputs [None {from=MaterialExpressionMultiply_0}]
   properties {transformSourceType=TRANSFORMSOURCE_World, transformType=TRANSFORM_Tangent}
  }
  MaterialExpressionVectorParameter_0 {class=MaterialExpressionVectorParameter, properties {group=AquaSynth, parameterName=HairColor_01, defaultValue=<rgba 0.2292 1 0.808 1>}}
  MaterialExpressionVectorParameter_1 {
   class=MaterialExpressionVectorParameter
   properties {group=AquaSynth, parameterName=HairColor_02, defaultValue=<rgba 0.1706 0.2548 0.6823 1>}
  }
  MaterialExpressionVectorParameter_2 {class=MaterialExpressionVectorParameter, properties {group=AquaSynth, parameterName=HairColor_03, defaultValue=<rgba 0.6212 0.4688 1 1>}}
  MaterialExpressionVertexInterpolator_16 {class=MaterialExpressionVertexInterpolator, inputs [VS {from=MaterialExpressionLocalPosition_1, output=XYZ}], properties {}}
  MaterialExpressionVertexNormalWS_0 {class=MaterialExpressionVertexNormalWS, inputs [], properties {}}
 }}
