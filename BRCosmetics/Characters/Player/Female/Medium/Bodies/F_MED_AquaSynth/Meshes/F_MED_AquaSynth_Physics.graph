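# Physics asset for the AquaSynth female medium player body cosmetic, covering the secondary-motion bones of the right sleeve ruffle.
# Capsule bodies sit on the ruffle base, the ruffle and spine_05.
# A single constraint joins the ruffle to its base with asymmetric swing limits and a twist limit.
# The spine_05 capsule has no constraint, which suggests it serves as a torso collider for the ruffle.
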PhysicsAsset /BRCosmetics/Characters/Player/Female/Medium/Bodies/F_MED_AquaSynth/Meshes/F_MED_AquaSynth_Physics
bodies {
 dyn_sleeve_ruffle_base_r [{shapeType=Capsule, length=0.505, radius=2.005, shapeName=dyn_sleeve_ruffle_base_r_capsule, rotation=<rot 0 -90 0>}]
 dyn_sleeve_ruffle_r [{
   shapeType=Capsule
   length=6.8864
   radius=4.1427
   shapeName=dyn_sleeve_ruffle_r_capsule
   center=<-4.1452 -0.4355 2.4027>
   rotation=<rot 0 -179.6044 -87.1952>
  }]
 spine_05 [{
   shapeType=Capsule
   length=3.9583
   radius=7.8155
   shapeName=spine_05_capsule
   center=<2.9367 -2.1876 -0.1441>
   rotation=<rot 0 -169.134 -176.3367>
  }]
}
constraints [{bone1Name=dyn_sleeve_ruffle_r, bone2Name=dyn_sleeve_ruffle_base_r, swing1LimitDegrees=15, swing2LimitDegrees=30, twistLimitDegrees=20}]
