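# Post-copy secondary-motion AnimBP for the female AquaSynth body cosmetic, animating ruffles, bracelets and cloth materials.
# Copied parent pose passes through a sleeve constraint, state-driven Control Rig, RigidBody, then Anim Dynamics for pants ruffles and bracelets.
# EventGraph converts skydive, crouch, emote and front-end flags into rig inputs and bracelet alpha, and feeds calf-driven WPO cloth, with skydive wind, to dynamic materials.
# Several AnimGraph nodes are disconnected from the output.
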AnimBlueprint /BRCosmetics/Characters/Player/Female/Medium/Bodies/F_MED_AquaSynth/Meshes/F_MED_AquaSynth_AnimBP
target_skeleton=&/F_MED_AquaSynth_Skeleton
graphs [
 AnimGraph {nodes [
   Anim Dynamics - Bone: dyn_sleeve_ruffle_r(Alpha=1, ExternalForce="0,0,0") {
    bAngularSpring=true
    bGravityOverrideInSimSpace=true
    bUseGravityOverride=true
    simulationSpace=RootRelative
    alphaBoolBlend {blendOption=Linear}
    boundBone {boneName=dyn_sleeve_ruffle_r}
    componentAppliedLinearAccClamp=<2000 2000 2000>
    componentLinearAccScale=<0.3 0.3 0.35>
    componentLinearVelScale=<0.3 0.3 0.35>
    gravityOverride {z=-980}
    physicsBodyDefinitions [{
      sphereCollisionRadius=0
      boundBone {boneName=dyn_sleeve_ruffle_r}
      boxExtents=<10 10 10>
      constraintSetup {angularLimitsMax=<10 10 10>, angularLimitsMin=<-10 -10 -10>, angularTarget=<0 0 0>}
      localJointOffset=<0 -1 0>
     }]
   }
   Anim Dynamics - Bone: dyn_pants_ruffles_r(Alpha=1, ExternalForce="0,0,0") {
    angularDampingOverride=0.99
    angularSpringConstant=350
    bAngularSpring=true
    bGravityOverrideInSimSpace=true
    bLinearSpring=true
    bOverrideAngularDamping=true
    bOverrideLinearDamping=true
    bUseGravityOverride=true
    linearDampingOverride=0.9999
    linearSpringConstant=25
    simulationSpace=BoneRelative
    alphaBoolBlend {blendOption=Linear}
    boundBone {boneName=dyn_pants_ruffles_r}
    componentAppliedLinearAccClamp=<3000 3000 3000>
    gravityOverride {z=-980}
    physicsBodyDefinitions [{
      sphereCollisionRadius=0
      boundBone {boneName=dyn_pants_ruffles_r}
      boxExtents=<30 50 40>
      constraintSetup {angularLimitsMax=<30 10 10>, angularLimitsMin=<-20 -10 -10>, angularTarget=<0 0 0>, linearAxesMin=<-2 0 0>}
      localJointOffset=<-4.5 0 0>
     }]
    relativeSpaceBone {boneName=dyn_pants_ruffles_base_r}
    simSpaceSettings {maxAngularAcceleration=150, maxAngularVelocity=5, simSpaceAngularAlpha=0.5, externalAngularVelocity=<1 1 1>}
   }
   n1 = Anim Dynamics - Bone: dyn_bracelets_1_l(Alpha=1, ExternalForce="0,0,0") {
    angularDampingOverride=0.999
    bAngularSpring=true
    bGravityOverrideInSimSpace=true
    bOverrideAngularDamping=true
    bUseGravityOverride=true
    simulationSpace=RootRelative
    alphaBoolBlend {blendOption=Linear}
    boundBone {boneName=dyn_bracelets_1_l}
    componentAppliedLinearAccClamp=<2000 2000 2000>
    componentLinearAccScale=<0.3 0.3 0.35>
    componentLinearVelScale=<0.3 0.3 0.35>
    gravityOverride {z=-980}
    physicsBodyDefinitions [{
      sphereCollisionRadius=0
      boundBone {boneName=dyn_bracelets_1_l}
      boxExtents=<10 10 10>
      constraintSetup {angularLimitsMax=<10 10 10>, angularLimitsMin=<-10 -10 -10>, angularTarget=<0 0 0>}
      localJointOffset=<0 -1 0>
     }]
   }
   n2 = Anim Dynamics - Bone: dyn_bracelets_2_l(ComponentPose=@n1, Alpha=1, ExternalForce="0,0,0") {
    angularDampingOverride=0.999
    bAngularSpring=true
    bGravityOverrideInSimSpace=true
    bOverrideAngularDamping=true
    bUseGravityOverride=true
    simulationSpace=RootRelative
    alphaBoolBlend {blendOption=Linear}
    boundBone {boneName=dyn_bracelets_2_l}
    componentAppliedLinearAccClamp=<2000 2000 2000>
    componentLinearAccScale=<0.3 0.3 0.35>
    componentLinearVelScale=<0.3 0.3 0.35>
    gravityOverride {z=-980}
    physicsBodyDefinitions [{
      sphereCollisionRadius=0
      boundBone {boneName=dyn_bracelets_2_l}
      boxExtents=<10 10 10>
      constraintSetup {angularLimitsMax=<10 10 10>, angularLimitsMin=<-10 -10 -10>, angularTarget=<0 0 0>}
      localJointOffset=<0 -1 0>
     }]
   }
   n3 = Anim Dynamics - Bone: dyn_bracelets_3_l(ComponentPose=@n2, Alpha=1, ExternalForce="0,0,0") {
    angularDampingOverride=0.999
    bAngularSpring=true
    bGravityOverrideInSimSpace=true
    bOverrideAngularDamping=true
    bUseGravityOverride=true
    simulationSpace=RootRelative
    alphaBoolBlend {blendOption=Linear}
    boundBone {boneName=dyn_bracelets_3_l}
    componentAppliedLinearAccClamp=<2000 2000 2000>
    componentLinearAccScale=<0.3 0.3 0.35>
    componentLinearVelScale=<0.3 0.3 0.35>
    gravityOverride {z=-980}
    physicsBodyDefinitions [{
      sphereCollisionRadius=0
      boundBone {boneName=dyn_bracelets_3_l}
      boxExtents=<10 10 10>
      constraintSetup {angularLimitsMax=<10 10 10>, angularLimitsMin=<-10 -10 -10>, angularTarget=<0 0 0>}
      localJointOffset=<0 -1 0>
     }]
   }
   n4 = Anim Dynamics - Bone: dyn_bracelets_4_l(ComponentPose=@n3, Alpha=1, ExternalForce="0,0,0") {
    angularDampingOverride=0.999
    bAngularSpring=true
    bGravityOverrideInSimSpace=true
    bOverrideAngularDamping=true
    bUseGravityOverride=true
    simulationSpace=RootRelative
    alphaBoolBlend {blendOption=Linear}
    boundBone {boneName=dyn_bracelets_4_l}
    componentAppliedLinearAccClamp=<2000 2000 2000>
    componentLinearAccScale=<0.3 0.3 0.35>
    componentLinearVelScale=<0.3 0.3 0.35>
    gravityOverride {z=-980}
    physicsBodyDefinitions [{
      sphereCollisionRadius=0
      boundBone {boneName=dyn_bracelets_4_l}
      boxExtents=<10 10 10>
      constraintSetup {angularLimitsMax=<10 10 10>, angularLimitsMin=<-10 -10 -10>, angularTarget=<0 0 0>}
      localJointOffset=<0 -1 0>
     }]
   }
   n5 = Anim Dynamics - Bone: dyn_bracelets_1_r(ComponentPose=@n4, Alpha=1, ExternalForce="0,0,0") {
    angularDampingOverride=0.999
    bAngularSpring=true
    bGravityOverrideInSimSpace=true
    bOverrideAngularDamping=true
    bUseGravityOverride=true
    simulationSpace=RootRelative
    alphaBoolBlend {blendOption=Linear}
    boundBone {boneName=dyn_bracelets_1_r}
    componentAppliedLinearAccClamp=<2000 2000 2000>
    componentLinearAccScale=<0.3 0.3 0.35>
    componentLinearVelScale=<0.3 0.3 0.35>
    gravityOverride {z=-980}
    physicsBodyDefinitions [{
      sphereCollisionRadius=0
      boundBone {boneName=dyn_bracelets_1_r}
      boxExtents=<10 10 10>
      constraintSetup {angularLimitsMax=<10 10 10>, angularLimitsMin=<-10 -10 -10>, angularTarget=<0 0 0>}
      localJointOffset=<0 -1 0>
     }]
   }
   n6 = Anim Dynamics - Bone: dyn_bracelets_2_r(ComponentPose=@n5, Alpha=1, ExternalForce="0,0,0") {
    angularDampingOverride=0.999
    bAngularSpring=true
    bGravityOverrideInSimSpace=true
    bOverrideAngularDamping=true
    bUseGravityOverride=true
    simulationSpace=RootRelative
    alphaBoolBlend {blendOption=Linear}
    boundBone {boneName=dyn_bracelets_2_r}
    componentAppliedLinearAccClamp=<2000 2000 2000>
    componentLinearAccScale=<0.3 0.3 0.35>
    componentLinearVelScale=<0.3 0.3 0.35>
    gravityOverride {z=-980}
    physicsBodyDefinitions [{
      sphereCollisionRadius=0
      boundBone {boneName=dyn_bracelets_2_r}
      boxExtents=<10 10 10>
      constraintSetup {angularLimitsMax=<10 10 10>, angularLimitsMin=<-10 -10 -10>, angularTarget=<0 0 0>}
      localJointOffset=<0 -1 0>
     }]
   }
   Anim Dynamics - Bone: dyn_bracelets_3_r(ComponentPose=@n6, Alpha=1, ExternalForce="0,0,0") {
    angularDampingOverride=0.999
    bAngularSpring=true
    bGravityOverrideInSimSpace=true
    bOverrideAngularDamping=true
    bUseGravityOverride=true
    simulationSpace=RootRelative
    alphaBoolBlend {blendOption=Linear}
    boundBone {boneName=dyn_bracelets_3_r}
    componentAppliedLinearAccClamp=<2000 2000 2000>
    componentLinearAccScale=<0.3 0.3 0.35>
    componentLinearVelScale=<0.3 0.3 0.35>
    gravityOverride {z=-980}
    physicsBodyDefinitions [{
      sphereCollisionRadius=0
      boundBone {boneName=dyn_bracelets_3_r}
      boxExtents=<10 10 10>
      constraintSetup {angularLimitsMax=<10 10 10>, angularLimitsMin=<-10 -10 -10>, angularTarget=<0 0 0>}
      localJointOffset=<0 -1 0>
     }]
   }
   n7 = "Transform (Modify) Bone - Bone: dyn_sleeve_ruffle_base_r"(Alpha=1, Translation="0,-2,0", Rotation="-30,0,0", Scale="1,1,1") {
    rotationMode=BMM_Additive
    scaleMode=BMM_Additive
    translationMode=BMM_Additive
    alphaBoolBlend {blendOption=Linear}
    boneToModify {boneName=dyn_sleeve_ruffle_base_r}
    rotation {pitch=-30}
    translation {y=-2}
   }
   Anim Dynamics - Bone: dyn_sleeve_ruffle_r(ComponentPose=@n7, Alpha=1, ExternalForce="0,0,0") {
    angularDampingOverride=1
    angularSpringConstant=5000
    bAngularSpring=true
    bGravityOverrideInSimSpace=true
    bOverrideAngularDamping=true
    bUseGravityOverride=true
    linearDampingOverride=0.999
    linearSpringConstant=25
    simulationSpace=BoneRelative
    alphaBoolBlend {blendOption=Linear}
    boundBone {boneName=dyn_sleeve_ruffle_r}
    componentAppliedLinearAccClamp=<6000 6000 6000>
    gravityOverride {z=-980}
    physicsBodyDefinitions [{
      sphereCollisionRadius=0
      boundBone {boneName=dyn_sleeve_ruffle_r}
      boxExtents=<30 30 30>
      constraintSetup {angularLimitsMax=<15 30 15>, angularLimitsMin=<-15 -30 -15>, angularTarget=<0 0 0>}
      localJointOffset=<1 0 0>
     }]
    relativeSpaceBone {boneName=dyn_sleeve_ruffle_base_r}
    simSpaceSettings {maxAngularAcceleration=150, maxAngularVelocity=5, externalAngularVelocity=<1 1 1>}
   }
   "Constraint - dyn_necklace_charm_base (PPPP)"!(
    ComponentPose
    Alpha=1
    ConstraintWeights_0=0.3424
    ConstraintWeights_1:float=0.2897
    ConstraintWeights_2:float=0.2443
    ConstraintWeights_3:float=0.1236
   ) -> (Pose) {
    boneToModify {boneName=dyn_necklace_charm_base}
    constraintSetup [{targetBone {boneName=neck_02}}, {targetBone {boneName=spine_05}}, {targetBone {boneName=neck_01}}, {targetBone {boneName=head}}]
    constraintWeights [1,1,1,1]
   }
   n8 = "Constraint - dyn_sleeve_ruffle_base_r (RR)"!(
    ComponentPose=Local To Component(LocalPose=Copy Pose From Mesh() {bUseAttachedParent=true})
    Alpha=1
    ConstraintWeights_0=0.5
    ConstraintWeights_1:float=0.5
   ) -> (Pose) {
    boneToModify {boneName=dyn_sleeve_ruffle_base_r}
    constraintSetup [{transformType=Rotation, targetBone {boneName=upperarm_r}}, {transformType=Rotation, targetBone {boneName=clavicle_r}}]
    constraintWeights [1,1]
   }
   n9 = Control Rig!(
    Source=Component To Local(ComponentPose=@n8)
    Alpha
    is_crouch:double=Get Dyn_Is_Crouch() -> (Dyn_Is_Crouch:double)
    is_skydive:double=Get Dyn_Is_Skydive() -> (Dyn_Is_Skydive:double)
    is_emote:double=Get Dyn_Is_Emote() -> (Dyn_Is_Emote:double)
    is_FE:double=Get Dyn_Is_FE() -> (Dyn_Is_FE:double)
   ) -> (Pose) {controlRigAssetReference {blueprintRigClass=&/F_MED_AquaSynth_CtrlRig.F_MED_AquaSynth_CtrlRig_C}}
   n10 = RigidBody(ComponentPose=Local To Component(LocalPose=@n9)) {
    bOverrideWorldGravity=true
    overridePhysicsAsset=&/F_MED_AquaSynth_Physics
    simulationSpace=BaseBoneSpace
    baseBoneRef {boneName=neck_01}
    componentAppliedLinearAccClamp=<2500 2500 2500>
    overrideWorldGravity {z=-980}
    simSpaceSettings {
     maxAngularAcceleration=45
     maxAngularVelocity=5
     maxLinearAcceleration=450
     maxLinearVelocity=100
     worldAlpha=0.5
     externalLinearDragV=<1 1 1>
    }
   }
   n11 = Anim Dynamics - Bone: dyn_pants_ruffles_l(ComponentPose=@n10, Alpha=1, ExternalForce="0,0,0") {
    angularDampingOverride=0.9999
    angularSpringConstant=250
    bAngularSpring=true
    bGravityOverrideInSimSpace=true
    bLinearSpring=true
    bOverrideAngularDamping=true
    bOverrideLinearDamping=true
    bUseGravityOverride=true
    linearDampingOverride=0.9999
    linearSpringConstant=60
    simulationSpace=RootRelative
    alphaBoolBlend {blendOption=Linear}
    boundBone {boneName=dyn_pants_ruffles_l}
    componentAppliedLinearAccClamp=<8000 8000 8000>
    gravityOverride {z=-980}
    physicsBodyDefinitions [{
      sphereCollisionRadius=0
      boundBone {boneName=dyn_pants_ruffles_l}
      boxExtents=<50 100 100>
      constraintSetup {angularLimitsMax=<30 5 5>, angularLimitsMin=<-30 -20 -20>, angularTarget=<0 0 0>, linearAxesMax=<3 0 0>}
      localJointOffset=<-2 0 0>
     }]
    relativeSpaceBone {boneName=dyn_pants_ruffles_base_l}
    simSpaceSettings {maxAngularAcceleration=45, maxAngularVelocity=10, simSpaceAngularAlpha=1, externalAngularVelocity=<1 1 1>}
   }
   n12 = Anim Dynamics - Bone: dyn_pants_ruffles_r(ComponentPose=@n11, Alpha=1, ExternalForce="0,0,0") {
    angularDampingOverride=0.9999
    angularSpringConstant=250
    bAngularSpring=true
    bGravityOverrideInSimSpace=true
    bLinearSpring=true
    bOverrideAngularDamping=true
    bOverrideLinearDamping=true
    bUseGravityOverride=true
    linearDampingOverride=0.9999
    linearSpringConstant=60
    simulationSpace=RootRelative
    alphaBoolBlend {blendOption=Linear}
    boundBone {boneName=dyn_pants_ruffles_r}
    componentAppliedLinearAccClamp=<8000 8000 8000>
    gravityOverride {z=-980}
    physicsBodyDefinitions [{
      sphereCollisionRadius=0
      boundBone {boneName=dyn_pants_ruffles_r}
      boxExtents=<50 100 100>
      constraintSetup {angularLimitsMax=<30 20 5>, angularLimitsMin=<-30 -10 -20>, angularTarget=<0 0 0>, linearAxesMin=<-3 0 0>}
      localJointOffset=<2 0 0>
     }]
    relativeSpaceBone {boneName=dyn_pants_ruffles_base_r}
    simSpaceSettings {maxAngularAcceleration=45, maxAngularVelocity=10, simSpaceAngularAlpha=1, externalAngularVelocity=<1 1 1>}
   }
   n13 = Get Dyn_Bracelets_alfa() -> (Dyn_Bracelets_alfa:float)
   n14 = Anim Dynamics - Bone: dyn_bracelets_1_l(ComponentPose=@n12, Alpha=@n13, ExternalForce="0,0,0") {
    angularDampingOverride=0.999
    angularSpringConstant=5
    bAngularSpring=true
    bGravityOverrideInSimSpace=true
    bOverrideAngularDamping=true
    bUseGravityOverride=true
    alphaBoolBlend {blendOption=Linear}
    boundBone {boneName=dyn_bracelets_1_l}
    componentAppliedLinearAccClamp=<3000 3000 3000>
    gravityOverride {z=-980}
    physicsBodyDefinitions [{
      sphereCollisionRadius=0
      boundBone {boneName=dyn_bracelets_1_l}
      boxExtents=<20 20 20>
      constraintSetup {angularLimitsMax=<10 2.5 10>, angularLimitsMin=<-10 -10 -10>, angularTarget=<0 0 0>}
      localJointOffset=<0 -1 0>
     }]
    simSpaceSettings {simSpaceAngularAlpha=1, externalAngularVelocity=<1 1 1>}
   }
   n15 = Anim Dynamics - Bone: dyn_bracelets_2_l(ComponentPose=@n14, Alpha=@n13, ExternalForce="0,0,0") {
    angularDampingOverride=0.999
    angularSpringConstant=5
    bAngularSpring=true
    bGravityOverrideInSimSpace=true
    bOverrideAngularDamping=true
    bUseGravityOverride=true
    alphaBoolBlend {blendOption=Linear}
    boundBone {boneName=dyn_bracelets_2_l}
    componentAppliedLinearAccClamp=<3000 3000 3000>
    gravityOverride {z=-980}
    physicsBodyDefinitions [{
      sphereCollisionRadius=0
      boundBone {boneName=dyn_bracelets_2_l}
      boxExtents=<20 20 20>
      constraintSetup {angularLimitsMax=<10 10 10>, angularLimitsMin=<-10 -10 -10>, angularTarget=<0 0 0>}
      localJointOffset=<0 -1 0>
     }]
    simSpaceSettings {simSpaceAngularAlpha=1, externalAngularVelocity=<1 1 1>}
   }
   n16 = Anim Dynamics - Bone: dyn_bracelets_3_l(ComponentPose=@n15, Alpha=@n13, ExternalForce="0,0,0") {
    angularDampingOverride=0.999
    angularSpringConstant=5
    bAngularSpring=true
    bGravityOverrideInSimSpace=true
    bOverrideAngularDamping=true
    bUseGravityOverride=true
    alphaBoolBlend {blendOption=Linear}
    boundBone {boneName=dyn_bracelets_3_l}
    componentAppliedLinearAccClamp=<3000 3000 3000>
    gravityOverride {z=-980}
    physicsBodyDefinitions [{
      sphereCollisionRadius=0
      boundBone {boneName=dyn_bracelets_3_l}
      boxExtents=<20 20 20>
      constraintSetup {angularLimitsMax=<10 10 10>, angularLimitsMin=<-10 -10 -10>, angularTarget=<0 0 0>}
      localJointOffset=<0 -1 0>
     }]
    simSpaceSettings {simSpaceAngularAlpha=1, externalAngularVelocity=<1 1 1>}
   }
   n17 = Anim Dynamics - Bone: dyn_bracelets_4_l(ComponentPose=@n16, Alpha=@n13, ExternalForce="0,0,0") {
    angularDampingOverride=0.999
    angularSpringConstant=5
    bAngularSpring=true
    bGravityOverrideInSimSpace=true
    bOverrideAngularDamping=true
    bUseGravityOverride=true
    alphaBoolBlend {blendOption=Linear}
    boundBone {boneName=dyn_bracelets_4_l}
    componentAppliedLinearAccClamp=<3000 3000 3000>
    gravityOverride {z=-980}
    physicsBodyDefinitions [{
      sphereCollisionRadius=0
      boundBone {boneName=dyn_bracelets_4_l}
      boxExtents=<20 20 20>
      constraintSetup {angularLimitsMax=<10 10 10>, angularLimitsMin=<-10 -10 -10>, angularTarget=<0 0 0>}
      localJointOffset=<0 -1 0>
     }]
    simSpaceSettings {simSpaceAngularAlpha=1, externalAngularVelocity=<1 1 1>}
   }
   n18 = Anim Dynamics - Bone: dyn_bracelets_1_r(ComponentPose=@n17, Alpha=@n13, ExternalForce="0,0,0") {
    angularDampingOverride=0.995
    angularSpringConstant=10
    bAngularSpring=true
    bGravityOverrideInSimSpace=true
    bOverrideAngularDamping=true
    bUseGravityOverride=true
    alphaBoolBlend {blendOption=Linear}
    boundBone {boneName=dyn_bracelets_1_r}
    componentAppliedLinearAccClamp=<3000 3000 3000>
    gravityOverride {z=-980}
    physicsBodyDefinitions [{
      sphereCollisionRadius=0
      boundBone {boneName=dyn_bracelets_1_r}
      boxExtents=<30 10 30>
      constraintSetup {angularLimitsMax=<10 10 5>, angularLimitsMin=<-10 -10 -10>, angularTarget=<0 0 0>}
      localJointOffset=<0 -1 0>
     }]
    simSpaceSettings {simSpaceAngularAlpha=1, externalAngularVelocity=<1 1 1>}
   }
   n19 = Anim Dynamics - Bone: dyn_bracelets_2_r(ComponentPose=@n18, Alpha=@n13, ExternalForce="0,0,0") {
    angularDampingOverride=0.995
    angularSpringConstant=10
    bAngularSpring=true
    bGravityOverrideInSimSpace=true
    bOverrideAngularDamping=true
    bUseGravityOverride=true
    alphaBoolBlend {blendOption=Linear}
    boundBone {boneName=dyn_bracelets_2_r}
    componentAppliedLinearAccClamp=<3000 3000 3000>
    gravityOverride {z=-980}
    physicsBodyDefinitions [{
      sphereCollisionRadius=0
      boundBone {boneName=dyn_bracelets_2_r}
      boxExtents=<30 12.5 30>
      constraintSetup {angularLimitsMax=<10 10 5>, angularLimitsMin=<-10 -10 -10>, angularTarget=<0 0 0>}
      localJointOffset=<0 -1 0>
     }]
    simSpaceSettings {simSpaceAngularAlpha=1, externalAngularVelocity=<1 1 1>}
   }
   n20 = Anim Dynamics - Bone: dyn_bracelets_3_r(ComponentPose=@n19, Alpha=@n13, ExternalForce="0,0,0") {
    angularDampingOverride=0.995
    angularSpringConstant=10
    bAngularSpring=true
    bGravityOverrideInSimSpace=true
    bOverrideAngularDamping=true
    bUseGravityOverride=true
    alphaBoolBlend {blendOption=Linear}
    boundBone {boneName=dyn_bracelets_3_r}
    componentAppliedLinearAccClamp=<3000 3000 3000>
    gravityOverride {z=-980}
    physicsBodyDefinitions [{
      sphereCollisionRadius=0
      boundBone {boneName=dyn_bracelets_3_r}
      boxExtents=<30 25 30>
      constraintSetup {angularLimitsMax=<10 10 5>, angularLimitsMin=<-10 -10 -10>, angularTarget=<0 0 0>}
      localJointOffset=<0 -1 0>
     }]
    simSpaceSettings {simSpaceAngularAlpha=1, externalAngularVelocity=<1 1 1>}
   }
   Output Pose(Result=Component To Local(ComponentPose=@n20)) {layerGroup=None}
  ]}
 EventGraph {nodes [
   n1 = Event BlueprintUpdateAnimation()
   > Set DeltaTimeX(DeltaTimeX:double=@n1.DeltaTimeX) -> (Output_Get:double)
   > n2 = Sequence!(execute) -> (then_0, then_1, then_2:Exec, then_3:Exec, then_4:Exec, then_5:Exec)
   > n3 = Branch(Condition=Get bIsSkyDiving())
   > Set Dyn_Is_Skydive(Dyn_Is_Skydive:double=1) -> (Output_Get:double)
   Set Dyn_Is_Skydive(execute=@n3.else, Dyn_Is_Skydive=0)
   n4 = Branch(execute=@n2.then_1, Condition=Get bIsCrouching())
   > Set Dyn_Is_Crouch(Dyn_Is_Crouch:double=1) -> (Output_Get:double)
   Set Dyn_Is_Crouch(execute=@n4.else, Dyn_Is_Crouch=0)
   n5 = Branch(execute=@n2.then_2, Condition=Get bIsPlayingEmote())
   > Set Dyn_Is_Emote(Dyn_Is_Emote:double=1) -> (Output_Get:double)
   Set Dyn_Is_Emote(execute=@n5.else, Dyn_Is_Emote=0)
   n6 = Branch(execute=@n2.then_3, Condition=Get bIsInFrontEnd())
   > Set Dyn_Is_FE(Dyn_Is_FE:double=1) -> (Output_Get:double)
   > Set Dyn_Bracelets_alfa(Dyn_Bracelets_alfa:float=0.8) -> (Output_Get:float)
   Set Dyn_Is_FE(execute=@n6.else, Dyn_Is_FE=0)
   > Set Dyn_Bracelets_alfa(Dyn_Bracelets_alfa=1)
   Set WPO Globals Function(
    execute=@n2.then_5
    Material=Get dynamic_mat_instance() -> (dynamic_mat_instance:Material Instance Dynamic Object Reference)
    Overlay Material=Get dynamic_overlay_mat() -> (dynamic_overlay_mat:Material Instance Dynamic Object Reference)
    Skeletal Mesh=Get Skeletal Mesh() -> (Skeletal Mesh:Skeletal Mesh Component Object Reference)
    "Local Noise Transform  Location"="0, 0, 0"
    Local Noise Offset="0, 0, 0"
   )
   > Set WPO Regional Function!(
    execute
    Material:Material Instance Dynamic Object Reference=Get dynamic_mat_instance()
    Overlay Material:Material Instance Dynamic Object Reference=Get dynamic_overlay_mat()
    Skeletal Mesh:Skeletal Mesh Component Object Reference=Get Skeletal Mesh()
    Velocity Transform:Name=calf_r
    Velocity Interp Speed:double=Get Velocity Interp Speed_reg1() -> (Velocity Interp Speed_reg1:double)
    Collision Transform:Name=calf_r
    Collision Transform Axis:EAxis Enum
    Invert Collision Transform Axis:bool=false
    Delta Time X:double=Get DeltaTimeX() -> (DeltaTimeX:double)
    Previous Velocity Transform Location:"Vector (by ref)"=Get Previous Velocity Transform Location_reg1() -> (Previous Velocity Transform Location_reg1:Vector)
    Previous Velocity Transform Velocity:"Vector (by ref)"=Get Previous Velocity Transform Velocity_reg1() -> (Previous Velocity Transform Velocity_reg1:Vector)
    Region:int=1
   ) -> (then)
   n7 = Get dynamic_mat_instance()
   n8 = Get dynamic_overlay_mat()
   > SetScalarParameterValue(self=[@n7, @n8], ParameterName=ClothSim_WindStrength, Value=Get wind_strength() -> (wind_strength:float))
   > SetScalarParameterValue(
    self=[@n8, @n7]
    ParameterName=ClothSim_TurbulenceNoiseTileDivisor
    Value=Get TurbulenceNoiseTileDivisor() -> (TurbulenceNoiseTileDivisor:float)
   )
   n9 = Branch(execute=@n2.then_4, Condition=Get bIsSkyDiving())
   > Set Velocity Interp Speed_reg1(Velocity Interp Speed_reg1:double=75) -> (Output_Get:double)
   > Set wind_strength(wind_strength:float=75) -> (Output_Get:float)
   > Set TurbulenceNoiseTileDivisor(TurbulenceNoiseTileDivisor:float=100) -> (Output_Get:float)
   Set Velocity Interp Speed_reg1(execute=@n9.else, Velocity Interp Speed_reg1=0)
   > Set wind_strength(wind_strength=0)
   > Set TurbulenceNoiseTileDivisor(TurbulenceNoiseTileDivisor=0)
   n10 = Event OnCharacterCustomizationLoaded()
   > n11 = Set Skeletal Mesh(Skeletal Mesh:Skeletal Mesh Component Object Reference=Get Array Item with EFort Custom Part Type(Array=@n10.CustomMeshParts)) -> (Output_Get:Skeletal Mesh Component Object Reference)
   > n12 = Get Dynamic Material Instance(SkeletalMeshComponent=@n11)
   > Set dynamic_mat_instance(dynamic_mat_instance:Material Instance Dynamic Object Reference=@n12) -> (Output_Get:Material Instance Dynamic Object Reference)
   > n13 = GetOrCreateDynamicOverlayMaterial(MeshComponent=Get Skeletal Mesh(), MaterialName=None)
   > Set dynamic_overlay_mat(dynamic_overlay_mat:Material Instance Dynamic Object Reference=@n13) -> (Output_Get:Material Instance Dynamic Object Reference)
  ]}
]
variables [
 DeltaTimeX {type=number}
 Dyn_Bracelets_alfa {type=number, default=1}
 Dyn_Is_Crouch {type=number}
 Dyn_Is_Emote {type=number, default=1}
 Dyn_Is_FE {type=number, default=1}
 Dyn_Is_Skydive {type=number, default=1}
 dynamic_mat_instance {type=MaterialInstanceDynamic}
 dynamic_overlay_mat {type=MaterialInstanceDynamic}
 Previous Velocity Transform Location_reg1 {type=Vector}
 Previous Velocity Transform Velocity_reg1 {type=Vector}
 Skeletal Mesh {type=SkeletalMeshComponent}
 TurbulenceNoiseTileDivisor {type=number}
 Velocity Interp Speed_reg1 {type=number, default=1}
 wind_strength {type=number}
]
