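# Control Rig for the AquaSynth female character's face accessory part, applying corrective offsets to hair-tail and necklace-charm base bones.
# A spherical pose reader on the head bone, inverted and scaled by dyn_bacpack (off by default), weights an offset of the hair tail base, likely for backpack clearance.
# A second transform modifier nudges and rotates the necklace charm base, weighted by dyn_FE (on by default).
# The pose reader's debug drawing is enabled.
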ControlRigBlueprint /BRCosmetics/Characters/Player/Female/Medium/Bodies/F_MED_AquaSynth/Meshes/Parts/F_MED_AquaSynth_FaceAcc_CtrlRig
graphs [
 RigVMFunctionLibrary {links [], nodes []}
 RigVMModel {nodes [
   RigUnit_BeginExecution = &RigUnit_BeginExecution()
   n1 = &RigUnit_SphericalPoseReader(
    ActiveRegionScaleFactors=(PositiveWidth=0.5,NegativeWidth=0.5,PositiveHeight=0.5,NegativeHeight=0.5)
    ActiveRegionSize=0.1
    Debug=(bDrawDebug=True,bDraw2D=False,bDrawLocalAxes=False,DebugScale=25,DebugSegments=20,DebugThickness=0.25)
    DriverAxis=(X=1,Y=0,Z=0)
    DriverItem=(Type=Bone,Name="head")
    FalloffRegionScaleFactors=(PositiveWidth=1,NegativeWidth=1,PositiveHeight=1,NegativeHeight=1)
    FalloffSize=0.2
    FlipHeightScaling=False
    FlipWidthScaling=False
    OptionalParentItem=(Type=Bone,Name="None")
    RotationOffset=(X=90,Y=0,Z=-75)
    ExecutePin=@RigUnit_BeginExecution.ExecutePin
   )
   Subtract = &RigVMFunction_MathDoubleSub(A=1, B=@n1.OutputParam)
   Multiply = &RigVMFunction_MathDoubleMul(A=@Subtract.Result, B=RigVMVariableNode() {variable=dyn_bacpack}.Value)
   > &RigUnit_ModifyTransforms(
    ItemToModify=((Item=(Type=Bone,Name="dyn_hair_tail_base"),Transform=(Rotation=(X=0,Y=0,Z=0,W=1),Translation=(X=-7.5,Y=0,Z=-15),Scale3D=(X=1,Y=1,Z=1))))
    Weight=@Multiply.Result
   )
   > &RigUnit_ModifyTransforms(
    ItemToModify=((Item=(Type=Bone,Name="dyn_necklace_charm_base"),Transform=(Rotation=(X=0,Y=0,Z=-0.2588,W=0.9659),Translation=(X=0,Y=-0.25,Z=0),Scale3D=(X=1,Y=1,Z=1))))
    Weight=RigVMVariableNode() {variable=dyn_FE}.Value
   )
  ]}
]
hierarchy [
 root {loc=<0 0 0>, rot=<0 0 0 1>}
 pelvis {parent=root, loc=<0 -0.5301 98.0571>, rot=<0.0127 0.707 -0.0127 -0.707>}
 spine_01 {parent=pelvis, loc=<6.1208 -0.9328 0>, rot=<0 0 0.034 -0.9994>}
 spine_02 {parent=spine_01, loc=<6.5997 -0.2095 0>, rot=<0 0 0.0238 -0.9997>}
 spine_03 {parent=spine_02, loc=<6.4914 -0.1237 0>, rot=<0 0 0.0149 -0.9999>}
 spine_04 {parent=spine_03, loc=<6.9598 0.0794 0>, rot=<0 0 -0.0132 -0.9999>}
 spine_05 {parent=spine_04, loc=<6.9987 0 0>, rot=<0 0 -0.0015 -1>}
 neck_01 {parent=spine_05, loc=<11.3138 0.1529 0>, rot=<0 0 -0.2003 -0.9797>}
 neck_02 {parent=neck_01, loc=<4.5627 0 0>, rot=<0 0 0.0504 -0.9987>}
 head {parent=neck_02, loc=<5.3914 0 0>, rot=<0 0 0.115 -0.9934>}
 dyn_hair_tail_base {parent=head, loc=<17.1833 -7.7095 0>, rot=<-0.6219 0.3598 0.2709 -0.6406>}
 dyn_hair_tail_1 {parent=dyn_hair_tail_base, loc=<4.149 -0.5185 0.0169>, rot=<0.6416 -0.0275 0.7643 0.058>}
 dyn_hair_tail_2 {parent=dyn_hair_tail_1, loc=<14.1421 0 0>, rot=<0 -0.2556 0 -0.9668>}
 dyn_hair_tail_3 {parent=dyn_hair_tail_2, loc=<13.66 0 0>, rot=<0 0.0828 0 -0.9966>}
 dyn_hair_tail_4 {parent=dyn_hair_tail_3, loc=<14.0867 0 0>, rot=<0 -0.2193 0 -0.9756>}
 dyn_hair_tail_5 {parent=dyn_hair_tail_4, loc=<15.3471 0 0>, rot=<0 0.2825 0 -0.9593>}
 dyn_hair_base_c {parent=head, loc=<9.4179 9.7254 0.0434>, rot=<-0.1497 -0.9724 0.1601 0.08>}
 dyn_hair_1_c {parent=dyn_hair_base_c, loc=<0 0 0>, rot=<0 0 0 1>}
 dyn_hair_2_c {parent=dyn_hair_1_c, loc=<3.7775 -0.2337 0.6359>, rot=<0 0.3555 0 -0.9347>}
 dyn_hair_3_c {parent=dyn_hair_1_c, loc=<2.1123 0.4795 2.0932>, rot=<0.2201 0.5228 0.0503 -0.822>}
 dyn_hair_base_r {parent=head, loc=<7.8595 7.4165 5.6055>, rot=<0.1777 -0.9555 -0.1146 0.2057>}
 dyn_hair_1_r {parent=dyn_hair_base_r, loc=<0 0 0>, rot=<0 0 0 1>}
 dyn_hair_2_r {parent=dyn_hair_1_r, loc=<4.6364 0 0>, rot=<0 0.2235 0 -0.9747>}
 dyn_hair_3_r {parent=dyn_hair_2_r, loc=<6.5155 0 0>, rot=<0 0 0 1>}
 dyn_hair_base_l {parent=head, loc=<7.7951 6.7962 -6.2527>, rot=<0.1361 -0.962 0.1876 -0.1441>}
 dyn_hair_1_l {parent=dyn_hair_base_l, loc=<0 0 0>, rot=<0 0 0 1>}
 dyn_hair_2_l {parent=dyn_hair_1_l, loc=<5.795 0 0>, rot=<0 -0.2392 0 -0.971>}
 dyn_hair_3_l {parent=dyn_hair_2_l, loc=<7.162 0 0>, rot=<0 0 0 1>}
 clavicle_l {parent=spine_05, loc=<5.4831 -0.1053 -3.2946>, rot=<0.0465 -0.7289 -0.009 -0.6829>}
 upperarm_l {parent=clavicle_l, loc=<9.8658 -1.2136 -0.4105>, rot=<0.0164 -0.3245 -0.0269 -0.9454>}
 lowerarm_l {parent=upperarm_l, loc=<26.465 -0.8345 0.259>, rot=<0.0036 0.0161 -0.2298 -0.9731>}
 clavicle_r {parent=spine_05, loc=<5.4831 -0.1053 3.2946>, rot=<0.7289 0.0465 -0.6829 0.009>}
 upperarm_r {parent=clavicle_r, loc=<-9.8658 1.2136 0.4105>, rot=<0.0164 -0.3245 -0.0269 -0.9454>}
 lowerarm_r {parent=upperarm_r, loc=<-26.465 0.8345 -0.259>, rot=<0.0036 0.0161 -0.2298 -0.9731>}
 dyn_sleeve_ruffle_base_r {parent=clavicle_r, loc=<-9.8658 1.2136 0.4105>, rot=<0.0164 -0.3245 -0.0269 -0.9454>}
 dyn_necklace_charm_base {parent=spine_05, loc=<11.379 5.387 -0.0014>, rot=<-0.0886 -0.9961 -0.0003 -0.0016>}
 dyn_necklace_charm_1 {parent=dyn_necklace_charm_base, loc=<0 0 0>, rot=<0 0 0 1>}
 dyn_necklace_charm_2 {parent=dyn_necklace_charm_1, loc=<1.567 0 0>, rot=<0 0 0 1>}
]
variables [
 dyn_FE {type=double, default=1}
 dyn_bacpack {type=double, default=0}
]
