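# Physics asset for the AquaSynth female medium player cosmetic's FaceAcc part, defining secondary-motion bodies for hair and a necklace charm.
# Capsules cover head, neck, spine, pelvis, clavicles and arms, plus dynamic hair, ponytail, charm and an unconstrained right sleeve-ruffle base.
# Limited swing/twist constraints link left, right and five-segment tail chains, generally loosening toward tips, plus the charm.
# Center hair branches, with segments 2 and 3 both constrained to segment 1.
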PhysicsAsset /BRCosmetics/Characters/Player/Female/Medium/Bodies/F_MED_AquaSynth/Meshes/Parts/F_MED_AquaSynth_FaceAcc_Physics
bodies {
 clavicle_l [{
   shapeType=Capsule
   length=6.255
   radius=7.505
   shapeName=clavicle_l_capsule
   center=<7.1927 -0.5333 -1.6959>
   rotation=<rot -0.0012 -89.9999 -79.9999>
  }]
 clavicle_r [{
   shapeType=Capsule
   length=6.755
   radius=7.005
   shapeName=clavicle_r_capsule
   center=<-6.7819 0 1.7129>
   rotation=<rot -0.0012 -89.9999 -79.9999>
  }]
 dyn_hair_1_c [{
   shapeType=Capsule
   length=2.8193
   radius=1.5697
   shapeName=dyn_hair_1_c_capsule
   center=<2.8723 0.4463 1.4742>
   rotation=<rot 47.4374 0.9076 12.5006>
  }]
 dyn_hair_1_l [{
   shapeType=Capsule
   length=3.9295
   radius=1.5046
   shapeName=dyn_hair_1_l_capsule
   center=<2.4409 -0.0243 0.5573>
   rotation=<rot -39.9675 93.6626 -87.2537>
  }]
 dyn_hair_1_r [{
   shapeType=Capsule
   length=3.5387
   radius=1.7721
   shapeName=dyn_hair_1_r_capsule
   center=<2.1265 0.6106 0.1287>
   rotation=<rot 9.5827 -98.1167 -87.1137>
  }]
 dyn_hair_2_c [{
   shapeType=Capsule
   length=6.5972
   radius=0.9875
   shapeName=dyn_hair_2_c_capsule
   center=<5.5919 -0.9774 -0.0577>
   rotation=<rot 0.4784 73.6321 -87.2865>
  }]
 dyn_hair_2_l [{
   shapeType=Capsule
   length=3.9211
   radius=1.2902
   shapeName=dyn_hair_2_l_capsule
   center=<4.0534 0.0443 0.1324>
   rotation=<rot 0 -84.9536 -80.951>
  }]
 dyn_hair_2_r [{shapeType=Capsule, length=2.755, radius=1.755, shapeName=dyn_hair_2_r_capsule, center=<3.3893 0.025 0.0731>, rotation=<rot -90 0.0019 0>}]
 dyn_hair_3_c [{
   shapeType=Capsule
   length=4.4692
   radius=0.8309
   shapeName=dyn_hair_3_c_capsule
   center=<5.7173 -1.0002 -0.7675>
   rotation=<rot 0 64.2059 -68.0843>
  }]
 dyn_hair_3_l [{
   shapeType=Capsule
   length=3.9906
   radius=1.4415
   shapeName=dyn_hair_3_l_capsule
   center=<3.4127 1.2629 2.0365>
   rotation=<rot -7.5311 100.7211 -51.4311>
  }]
 dyn_hair_3_r [{
   shapeType=Capsule
   length=5.1919
   radius=1.1002
   shapeName=dyn_hair_3_r_capsule
   center=<4.1309 1.7228 0.3532>
   rotation=<rot -2.0001 116.2888 -84.4941>
  }]
 dyn_hair_base_c [{shapeType=Capsule, length=0.25, radius=1.505, shapeName=dyn_hair_base_c_capsule, rotation=<rot 0 -90 0>}]
 dyn_hair_base_l [{shapeType=Capsule, length=0.25, radius=1.505, shapeName=dyn_hair_base_l_capsule, rotation=<rot 0 -90 0>}]
 dyn_hair_base_r [{shapeType=Capsule, length=0.25, radius=1.505, shapeName=dyn_hair_base_r_capsule, rotation=<rot 0 -90 0>}]
 dyn_hair_tail_1 [{
   shapeType=Capsule
   length=7.8559
   radius=5.2001
   shapeName=dyn_hair_tail_1_capsule
   center=<7.1186 -1.0853 2.0209>
   rotation=<rot -21.2029 -93.6737 -92.4619>
  }]
 dyn_hair_tail_2 [
  {
   shapeType=Capsule
   length=6.861
   radius=7.0678
   shapeName=dyn_hair_tail_2_capsule
   center=<10.1642 -0.2543 1.4854>
   rotation=<rot -0.5663 -89.7269 -73.9106>
  }
  {
   shapeType=Capsule
   length=10.75
   radius=3
   shapeName=dyn_hair_tail_2_capsule
   center=<11.3606 4.9733 -4.0724>
   rotation=<rot -74.9998 -180 -180>
  }
 ]
 dyn_hair_tail_3 [
  {
   shapeType=Capsule
   length=7.5
   radius=5
   shapeName=dyn_hair_tail_3_capsule
   center=<7.356 0.0004 1.3298>
   rotation=<rot 1.2922 -85.1695 -104.941>
  }
  {
   shapeType=Capsule
   length=6.75
   radius=2.5
   shapeName=dyn_hair_tail_3_capsule
   center=<17.1126 2.1917 -3.9088>
   rotation=<rot -64.0155 -150.6428 155.9882>
  }
 ]
 dyn_hair_tail_4 [
  {
   shapeType=Capsule
   length=5.5239
   radius=5.2407
   shapeName=dyn_hair_tail_4_capsule
   center=<8.1571 0.0086 2.765>
   rotation=<rot -0.4777 -95.0202 -83.6853>
  }
  {
   shapeType=Capsule
   length=6.5
   radius=2.25
   shapeName=dyn_hair_tail_4_capsule
   center=<9.3625 -4.351 2.616>
   rotation=<rot -69.9999 145.0002 179.9998>
  }
 ]
 dyn_hair_tail_5 [{
   shapeType=Capsule
   length=6.2676
   radius=3.4284
   shapeName=dyn_hair_tail_5_capsule
   center=<5.0078 0.0267 1.4606>
   rotation=<rot 0 -84.5078 -96.563>
  }]
 dyn_hair_tail_base [{shapeType=Capsule, length=0.25, radius=1.505, shapeName=dyn_hair_tail_base_capsule, rotation=<rot 0 -90 0>}]
 dyn_necklace_charm_1 [{
   shapeType=Capsule
   length=1.01
   radius=1
   shapeName=dyn_necklace_charm_1_capsule
   center=<0.9546 -0.013 -0.0029>
   rotation=<rot 0 -90.0851 -89.9965>
  }]
 dyn_necklace_charm_2 [{
   shapeType=Capsule
   length=1.5984
   radius=1.5
   shapeName=dyn_necklace_charm_2_capsule
   center=<4.5552 0.4245 -0.0205>
   rotation=<rot 0 -86.41 0.1776>
  }]
 dyn_necklace_charm_base [{shapeType=Capsule, length=0.505, radius=0.505, shapeName=dyn_necklace_charm_base_capsule, rotation=<rot 0 -90 0>}]
 dyn_sleeve_ruffle_base_r [{
   shapeType=Capsule
   length=3.75
   radius=6.75
   shapeName=dyn_sleeve_ruffle_base_r_capsule
   center=<-1.6375 0.5566 -2.8572>
   rotation=<rot 90 -89.9991 0>
  }]
 head [{shapeType=Capsule, length=7, radius=6.4684, shapeName=head_capsule, center=<3.8558 0.9042 0.034>, rotation=<rot -0.6905 -118.6801 -89.0137>}]
 lowerarm_l [{shapeType=Capsule, length=32, radius=4.25, shapeName=lowerarm_l_capsule, center=<14.8341 0.0001 0>, rotation=<rot -0.0002 -89.9998 89.9999>}]
 lowerarm_r [{
   shapeType=Capsule
   length=31.5
   radius=4.75
   shapeName=lowerarm_r_capsule
   center=<-13.6214 -0.0001 0>
   rotation=<rot -0.0002 -89.9998 89.9999>
  }]
 neck_01 [{
   shapeType=Capsule
   length=5.059
   radius=3.9766
   shapeName=neck_01_capsule
   center=<2.1098 0.4911 0.1014>
   rotation=<rot 75.9947 -87.7038 92.2284>
  }]
 pelvis [{shapeType=Capsule, length=11.01, radius=11.75, shapeName=pelvis_capsule, rotation=<rot 0 -90 0>}]
 spine_03 [{shapeType=Capsule, length=10, radius=7.75, shapeName=spine_03_capsule, rotation=<rot 0 -90 0>}]
 spine_04 [{shapeType=Capsule, length=8.005, radius=10.255, shapeName=spine_04_capsule, center=<0 1.1098 0>, rotation=<rot 0 -90 0>}]
 spine_05 [{shapeType=Capsule, length=13, radius=7.25, shapeName=spine_05_capsule, center=<3.7685 -1.8193 0>, rotation=<rot 0 -90 0>}]
 upperarm_l [{shapeType=Capsule, length=16.255, radius=5.005, shapeName=upperarm_l_capsule, center=<13.3695 0.0001 0>, rotation=<rot -0.0006 -89.9995 -90>}]
 upperarm_r [{
   shapeType=Capsule
   length=16.255
   radius=5.005
   shapeName=upperarm_r_capsule
   center=<-13.0944 -0.0001 0>
   rotation=<rot -0.0006 -89.9995 -90>
  }]
}
constraints [
 {bone1Name=dyn_hair_2_c, bone2Name=dyn_hair_1_c, swing1LimitDegrees=10, swing2LimitDegrees=12, twistLimitDegrees=2}
 {bone1Name=dyn_hair_3_c, bone2Name=dyn_hair_1_c, swing1LimitDegrees=10, swing2LimitDegrees=12, twistLimitDegrees=2}
 {bone1Name=dyn_hair_base_c, bone2Name=dyn_hair_1_c, swing1LimitDegrees=10, swing2LimitDegrees=10, twistLimitDegrees=2}
 {bone1Name=dyn_hair_1_l, bone2Name=dyn_hair_base_l, swing1LimitDegrees=5, swing2LimitDegrees=5, twistLimitDegrees=2}
 {bone1Name=dyn_hair_2_l, bone2Name=dyn_hair_1_l, swing1LimitDegrees=10, swing2LimitDegrees=10, twistLimitDegrees=5}
 {bone1Name=dyn_hair_3_l, bone2Name=dyn_hair_2_l, swing1LimitDegrees=20, swing2LimitDegrees=20, twistLimitDegrees=10}
 {bone1Name=dyn_hair_1_r, bone2Name=dyn_hair_base_r, swing1LimitDegrees=5, swing2LimitDegrees=5, twistLimitDegrees=2}
 {bone1Name=dyn_hair_2_r, bone2Name=dyn_hair_1_r, swing1LimitDegrees=15, swing2LimitDegrees=20, twistLimitDegrees=5}
 {bone1Name=dyn_hair_3_r, bone2Name=dyn_hair_2_r, swing1LimitDegrees=20, swing2LimitDegrees=30, twistLimitDegrees=10}
 {bone1Name=dyn_hair_tail_1, bone2Name=dyn_hair_tail_base, swing1LimitDegrees=10, swing2LimitDegrees=20, twistLimitDegrees=15}
 {bone1Name=dyn_hair_tail_2, bone2Name=dyn_hair_tail_1, swing1LimitDegrees=15, swing2LimitDegrees=20, twistLimitDegrees=15}
 {bone1Name=dyn_hair_tail_3, bone2Name=dyn_hair_tail_2, swing1LimitDegrees=15, swing2LimitDegrees=20, twistLimitDegrees=15}
 {bone1Name=dyn_hair_tail_4, bone2Name=dyn_hair_tail_3, swing1LimitDegrees=20, swing2LimitDegrees=25, twistLimitDegrees=20}
 {bone1Name=dyn_hair_tail_5, bone2Name=dyn_hair_tail_4, swing1LimitDegrees=20, swing2LimitDegrees=30, twistLimitDegrees=20}
 {bone1Name=dyn_necklace_charm_base, bone2Name=dyn_necklace_charm_1, swing1LimitDegrees=30, swing2LimitDegrees=10, twistLimitDegrees=5}
 {bone1Name=dyn_necklace_charm_2, bone2Name=dyn_necklace_charm_1, swing1LimitDegrees=30, swing2LimitDegrees=30, twistLimitDegrees=15}
]
